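# Material function for the SheerWill cape that outputs an animated emote reveal mask and its glowing edge color.
# A sphere mask around the offset pre-skinned local position, distorted by cloud noise, grows through three radius stages.
# The SheerWillEmoteAnimation parameter, scaled by speed offsets, is split into saturated stage alphas for the radius and noise strength blends.
# A smoothstep of the mask isolates the edge, which is tinted and modulated by a sine of the animation.
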MaterialFunction /BRCosmetics/Accessories/FORT_Capes/Cape_M_MED_SheerWillUniversal/Materials/MF/MF_SheerWill_Cape_EmoteMask
graph {nodes [
  n1 = TextureSample(UVs=Multiply(A=TextureCoordinate(), B=ScalarParameter() {defaultValue=2.8593, parameterName=Leg_L_tiling_stage2})) {samplerSource=SSM_Wrap_WorldGroupSettings, samplerType=SAMPLERTYPE_LinearColor, texture=/Game/Global/Textures/Default/Noise/T_CloudNoise_01}
  n2 = Multiply(
   A=LinearInterpolate(Alpha=@n1.B) {constA=-1}
   B=LinearInterpolate(
    A=ScalarParameter() {defaultValue=3.6713, parameterName=Leg_L_Strength_stage1}
    B=ScalarParameter() {defaultValue=29.3712, parameterName=Leg_L_Strength_stage2}
    Alpha=NamedRerouteUsage()
   )
  )
  n3 = NamedRerouteUsage()
  n4 = LinearInterpolate(
   A=LinearInterpolate(B=ScalarParameter() {defaultValue=76, parameterName=Leg_L_Radius_stage1}, Alpha=NamedRerouteUsage())
   B=ScalarParameter() {defaultValue=111.5165, parameterName=Leg_L_Radius_stage2}
   Alpha=NamedRerouteUsage()
  )
  n5 = ScalarParameter() {defaultValue=150, parameterName=Leg_L_Radius_stage3}
  n6 = NamedRerouteUsage()
  n7 = NamedRerouteDeclaration(None=SphereMask(
    A=Add(A=@n2, B=@n3)
    B=Constant()
    Radius=LinearInterpolate(A=@n4, B=@n5, Alpha=@n6)
    Hardness=ScalarParameter() {defaultValue=0.97, parameterName=Leg_L_Hardness}
   )) {name=OutMask, nodeColor=<rgba 0 0.5547 1 1>}
  FunctionOutput(None=@n7) {outputName=OutMask}
  n8 = Constant() {r=0.6143}
  n9 = Constant()
  n10 = Multiply(A=Multiply(A=SmoothStep(Min=@n8, Max=@n9, Value=@n7), B=@n7), B=ScalarParameter() {defaultValue=5, parameterName=GlowEdgeIntensity}) {constB=0.5}
  n11 = ScalarParameter() {group=SheerWill_BlendMask, parameterName=SheerWillEmoteAnimation}
  n12 = MaterialFunctionCall() {materialFunction=&Pi}
  n13 = NamedRerouteDeclaration(None=Sine(None=Multiply(A=@n11, B=@n12.Result)) {period=6.283}) {name=Sine_Stage2_3, nodeColor=<rgba 0 1 0.6016 1>}
  n14 = LinearInterpolate(A=ScalarParameter() {defaultValue=0.015, parameterName=GlowEdgeIntensityMIn}, Alpha=Max(A=Saturate(None=@n13))) {constA=0.1}
  FunctionOutput(None=Multiply(A=Multiply(A=@n10, B=@n14), B=VectorParameter() {parameterName=EmoteEdgeColor, defaultValue=<rgba 2.081 0.3564 3 1>}.RGB)) {outputName=EmoteEdgeColor}
  n15 = Multiply(A=Multiply(
    A=ScalarParameter() {group=SheerWill_BlendMask, parameterName=SheerWillEmoteAnimation}
    B=ScalarParameter() {defaultValue=1, group=SheerWill_BlendMask, parameterName=EmoteAnimation_SpeedOffset}
   )) {constB=4}
  n16 = Subtract(A=@n15)
  n17 = ScalarParameter() {defaultValue=1, parameterName=SpeedOffset}
  NamedRerouteDeclaration(None=Saturate(None=Multiply(A=@n16, B=@n17))) {name=Animation_Stage1_2, nodeColor=<rgba 0.7344 0 1 1>}
  NamedRerouteDeclaration(None=Saturate(None=Multiply(A=@n15, B=@n17))) {name=Animation_Stage0_1, nodeColor=<rgba 0.7344 0 1 1>}
  n18 = Subtract(A=@n15) {constB=2}
  NamedRerouteDeclaration(None=Saturate(None=Multiply(A=@n18, B=@n17))) {name=Animation_Stage2_3, nodeColor=<rgba 0.7344 0 1 1>}
  n19 = LocalPosition() {localOrigin=InstancePreSkinning}
  NamedRerouteDeclaration(None=Add(
    A=VertexInterpolator(VS=@n19.XYZ).PS
    B=VectorParameter() {group=SheerWill_BlendMask, parameterName=Emote_Position_Offset, defaultValue=<rgba 0 0 100 1>}.RGB
   )) {name=PreSkinnedLocalPosition, nodeColor=<rgba 0 1 0.0625 1>}
 ]}
