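# Material function that blends the SheerWill cape's material into an emote look with its own opacity mask and glowing edge.
# An emote mask function drives a switch to black, rough attributes with emissive built from tinted Fresnel, a SheerWill Fresnel mask and edge color.
# Opacity lerps toward a mask texture and fades out with the Emote_SheerPhaseCape scalar.
# That scalar plus the emote mask, saturated, sets the blend over the input attributes.
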MaterialFunction /BRCosmetics/Accessories/FORT_Capes/Cape_M_MED_SheerWillUniversal/Materials/MF/MF_SheerWill_Cape_EmoteBlendMaterial_opacity
graph {nodes [
  n1 = MaterialFunctionCall() {materialFunction=&/MF_SheerWill_Cape_EmoteMask}
  n2 = TextureSampleParameter2D() {parameterName=Emote_OPacityMask, samplerType=SAMPLERTYPE_Masks, texture=&T_BlackMask}
  n3 = ScalarParameter() {parameterName=Emote_SheerPhaseCape}
  n4 = FunctionInput() {bUsePreviewValueAsDefault=true, inputType=FunctionInput_MaterialAttributes}
  n5 = SetMaterialAttributes!(
   MaterialAttributes=@n4
   BaseColor=Constant()
   EmissiveColor=Add(A=NamedRerouteUsage(), B=@n1.EmoteEdgeColor)
   Metallic
   Specular=Constant() {r=0.1}
   Roughness=Constant() {r=1}
   OpacityMask=Multiply(A=LinearInterpolate(B=@n2.R, Alpha=@n1.OutMask) {constA=1}, B=OneMinus(None=@n3))
  ) {attributeSetTypes [
    #1
    #2
    #3
    #4
    #5
    #6
   ]}
  n6 = Add(A=@n3, B=@n1.OutMask)
  FunctionOutput(None=BlendMaterialAttributes(A=@n4, B=@n5, Alpha=Saturate(None=@n6)))
  Multiply()
  n7 = Power(Base=Fresnel(ExponentIn=ScalarParameter() {defaultValue=2.5, parameterName=Emote_FresnelExponentln}))
  n8 = VectorParameter() {parameterName=Emote_FresnelColor, defaultValue=<rgba 0.702 0.2575 1 1>}
  n9 = Multiply(
   A=ComponentMask(None=MaterialFunctionCall() {materialFunction=/BRCosmetics/Characters/Player/Male/Medium/Bodies/M_MED_SheerWill_Phase/Material/MF_SheerWill_FresnelMask}.SheerWill_FresnelMask) {r=true}
   B=@n8.RGB
  )
  NamedRerouteDeclaration(None=Add(A=Multiply(A=@n7, B=@n8.RGB), B=Multiply(A=@n9) {constB=0.6})) {name=OutEmissiveEmote, nodeColor=<rgba 0.6641 0 1 1>}
 ]}
