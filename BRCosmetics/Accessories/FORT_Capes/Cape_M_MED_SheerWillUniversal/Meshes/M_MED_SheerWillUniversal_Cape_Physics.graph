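# Physics asset for the SheerWillUniversal medium male cape, simulating it as four bone chains (l, cl, cr, r) hung from spine_05.
# Each chain runs from a small master body through seven identical capsules, linked by tight swing and twist limits.
# Capsules on the spine, pelvis, legs and upper arms form the body shapes, joined by free angular constraints.
# An oversized root capsule sits below the origin, apparently as a ground collider.
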PhysicsAsset /BRCosmetics/Accessories/FORT_Capes/Cape_M_MED_SheerWillUniversal/Meshes/M_MED_SheerWillUniversal_Cape_Physics
bodies {
 calf_l [{shapeType=Capsule, length=37.2374, radius=16.5, shapeName=calf_l_capsule, center=<-13.3998 -3.4567 0>, rotation=<rot 90 0 0>}]
 calf_r [{shapeType=Capsule, length=37.2374, radius=16.5, shapeName=calf_r_capsule, center=<13.3998 3.4567 0>, rotation=<rot 90 0 0>}]
 dyn_cape_cl_01 [{shapeType=Capsule, length=15.7494, radius=2.7476, shapeName=dyn_cape_cl_01_capsule, center=<13.3148 0 0>, rotation=<rot 90 0 0>}]
 dyn_cape_cl_02 [{shapeType=Capsule, length=15.7494, radius=2.7476, shapeName=dyn_cape_cl_02_capsule, center=<13.3148 0 0>, rotation=<rot 90 0 0>}]
 dyn_cape_cl_03 [{shapeType=Capsule, length=15.7494, radius=2.7476, shapeName=dyn_cape_cl_03_capsule, center=<13.3148 0 0>, rotation=<rot 90 0 0>}]
 dyn_cape_cl_04 [{shapeType=Capsule, length=15.7494, radius=2.7476, shapeName=dyn_cape_cl_04_capsule, center=<13.3148 0 0>, rotation=<rot 90 0 0>}]
 dyn_cape_cl_05 [{shapeType=Capsule, length=15.7494, radius=2.7476, shapeName=dyn_cape_cl_05_capsule, center=<13.3148 0 0>, rotation=<rot 90 0 0>}]
 dyn_cape_cl_06 [{shapeType=Capsule, length=15.7494, radius=2.7476, shapeName=dyn_cape_cl_06_capsule, center=<13.3148 0 0>, rotation=<rot 90 0.1419 0>}]
 dyn_cape_cl_07 [{shapeType=Capsule, length=15.7494, radius=2.7476, shapeName=dyn_cape_cl_07_capsule, center=<13.3148 0 0>, rotation=<rot 90 0 0>}]
 dyn_cape_cl_MSTR [{shapeType=Capsule, length=1, radius=1, shapeName=dyn_cape_cl_MSTR_capsule}]
 dyn_cape_cr_01 [{shapeType=Capsule, length=15.7494, radius=2.7476, shapeName=dyn_cape_cr_01_capsule, center=<13.3148 0 0>, rotation=<rot 90 0 0>}]
 dyn_cape_cr_02 [{shapeType=Capsule, length=15.7494, radius=2.7476, shapeName=dyn_cape_cr_02_capsule, center=<13.3148 0 0>, rotation=<rot 90 0 0>}]
 dyn_cape_cr_03 [{shapeType=Capsule, length=15.7494, radius=2.7476, shapeName=dyn_cape_cr_03_capsule, center=<13.3148 0 0.1882>, rotation=<rot 90 0 0>}]
 dyn_cape_cr_04 [{shapeType=Capsule, length=15.7494, radius=2.7476, shapeName=dyn_cape_cr_04_capsule, center=<13.3148 0 0>, rotation=<rot 90 0 0>}]
 dyn_cape_cr_05 [{shapeType=Capsule, length=15.7494, radius=2.7476, shapeName=dyn_cape_cr_05_capsule, center=<13.3148 0 0>, rotation=<rot 90 0 0>}]
 dyn_cape_cr_06 [{shapeType=Capsule, length=15.7494, radius=2.7476, shapeName=dyn_cape_cr_06_capsule, center=<13.3148 0 0>, rotation=<rot 90 0 0>}]
 dyn_cape_cr_07 [{shapeType=Capsule, length=15.7494, radius=2.7476, shapeName=dyn_cape_cr_07_capsule, center=<13.3148 0 0>, rotation=<rot 90 0 0>}]
 dyn_cape_cr_MSTR [{shapeType=Capsule, length=1, radius=1, shapeName=dyn_cape_cr_MSTR_capsule}]
 dyn_cape_l_01 [{shapeType=Capsule, length=15.7494, radius=2.7476, shapeName=dyn_cape_l_01_capsule, center=<13.3148 0 0>, rotation=<rot 90 0 0>}]
 dyn_cape_l_02 [{shapeType=Capsule, length=15.7494, radius=2.7476, shapeName=dyn_cape_l_02_capsule, center=<13.3148 0 0>, rotation=<rot 90 0 0>}]
 dyn_cape_l_03 [{shapeType=Capsule, length=15.7494, radius=2.7476, shapeName=dyn_cape_l_03_capsule, center=<13.3148 0 0>, rotation=<rot 90 0 0>}]
 dyn_cape_l_04 [{shapeType=Capsule, length=15.7494, radius=2.7476, shapeName=dyn_cape_l_04_capsule, center=<13.3148 0 0>, rotation=<rot 90 0 0>}]
 dyn_cape_l_05 [{shapeType=Capsule, length=15.7494, radius=2.7476, shapeName=dyn_cape_l_05_capsule, center=<13.3148 0 0>, rotation=<rot 90 0 0>}]
 dyn_cape_l_06 [{shapeType=Capsule, length=15.7494, radius=2.7476, shapeName=dyn_cape_l_06_capsule, center=<13.3148 0 0>, rotation=<rot 90 0 0>}]
 dyn_cape_l_07 [{shapeType=Capsule, length=15.7494, radius=2.7476, shapeName=dyn_cape_l_07_capsule, center=<13.3148 0 0>, rotation=<rot 90 0 0>}]
 dyn_cape_l_MSTR [{shapeType=Capsule, length=1, radius=1, shapeName=dyn_cape_l_MSTR_capsule}]
 dyn_cape_r_01 [{shapeType=Capsule, length=15.7494, radius=2.7476, shapeName=dyn_cape_r_01_capsule, center=<13.3148 0 0>, rotation=<rot 90 0 0>}]
 dyn_cape_r_02 [{shapeType=Capsule, length=15.7494, radius=2.7476, shapeName=dyn_cape_r_02_capsule, center=<13.3148 0 0>, rotation=<rot 90 0 0>}]
 dyn_cape_r_03 [{shapeType=Capsule, length=15.7494, radius=2.7476, shapeName=dyn_cape_r_03_capsule, center=<13.3148 0 0>, rotation=<rot 90 0 0>}]
 dyn_cape_r_04 [{shapeType=Capsule, length=15.7494, radius=2.7476, shapeName=dyn_cape_r_04_capsule, center=<13.3148 0 0.1882>, rotation=<rot 90 0 0>}]
 dyn_cape_r_05 [{shapeType=Capsule, length=15.7494, radius=2.7476, shapeName=dyn_cape_r_05_capsule, center=<13.3148 0 0>, rotation=<rot 90 0 0>}]
 dyn_cape_r_06 [{shapeType=Capsule, length=15.7494, radius=2.7476, shapeName=dyn_cape_r_06_capsule, center=<13.3148 0 0>, rotation=<rot 90 0 0>}]
 dyn_cape_r_07 [{shapeType=Capsule, length=15.7494, radius=2.7476, shapeName=dyn_cape_r_07_capsule, center=<13.3148 0 0>, rotation=<rot 90 0 0>}]
 dyn_cape_r_MSTR [{shapeType=Capsule, length=1, radius=1, shapeName=dyn_cape_r_MSTR_capsule}]
 pelvis [{shapeType=Capsule, length=7.3953, radius=17, shapeName=pelvis_capsule, rotation=<rot 0 -90 0>}]
 root [{shapeType=Capsule, length=232.7972, radius=326.3409, shapeName=root_capsule, center=<0 -73.1274 -326.3191>, rotation=<rot 90 -180 0>}]
 spine_03 [{shapeType=Capsule, length=0.505, radius=14.8869, shapeName=spine_03_capsule, rotation=<rot 0 -90 0>}]
 spine_05 [{shapeType=Capsule, length=11, radius=13, shapeName=spine_05_capsule}]
 thigh_l [{shapeType=Capsule, length=28, radius=16, shapeName=thigh_l_capsule, center=<-16.4237 0 0>, rotation=<rot 90 0 0>}]
 thigh_r [{shapeType=Capsule, length=28, radius=16, shapeName=thigh_r_capsule, center=<16.4237 0 0>, rotation=<rot 90 0 0>}]
 upperarm_l [{shapeType=Capsule, length=20.5, radius=8.5, shapeName=upperarm_l_capsule, center=<16.693 0 0>, rotation=<rot 90 0 0>}]
 upperarm_r [{shapeType=Capsule, length=20.5, radius=8.5, shapeName=upperarm_r_capsule, center=<-16.693 0 0>, rotation=<rot 90 0 0>}]
}
constraints [
 {
  bone1Name=spine_05
  bone2Name=spine_03
  swing1LimitDegrees=45
  swing1Motion=Free
  swing2LimitDegrees=45
  swing2Motion=Free
  twistLimitDegrees=45
  twistMotion=Free
 }
 {
  bone1Name=spine_03
  bone2Name=pelvis
  swing1LimitDegrees=45
  swing1Motion=Free
  swing2LimitDegrees=45
  swing2Motion=Free
  twistLimitDegrees=45
  twistMotion=Free
 }
 {bone1Name=dyn_cape_r_07, bone2Name=dyn_cape_r_06, swing1LimitDegrees=10, swing2LimitDegrees=15, twistLimitDegrees=8}
 {bone1Name=dyn_cape_cl_MSTR, bone2Name=spine_05, swing1LimitDegrees=10, swing2LimitDegrees=15, twistLimitDegrees=8}
 {bone1Name=dyn_cape_cl_01, bone2Name=dyn_cape_cl_MSTR, swing1LimitDegrees=10, swing2LimitDegrees=15, twistLimitDegrees=8}
 {bone1Name=dyn_cape_cl_02, bone2Name=dyn_cape_cl_01, swing1LimitDegrees=10, swing2LimitDegrees=15, twistLimitDegrees=8}
 {bone1Name=dyn_cape_cl_03, bone2Name=dyn_cape_cl_02, swing1LimitDegrees=10, swing2LimitDegrees=15, twistLimitDegrees=8}
 {bone1Name=dyn_cape_cl_04, bone2Name=dyn_cape_cl_03, swing1LimitDegrees=10, swing2LimitDegrees=15, twistLimitDegrees=8}
 {bone1Name=dyn_cape_cl_05, bone2Name=dyn_cape_cl_04, swing1LimitDegrees=10, swing2LimitDegrees=15, twistLimitDegrees=8}
 {bone1Name=dyn_cape_cl_06, bone2Name=dyn_cape_cl_05, swing1LimitDegrees=10, swing2LimitDegrees=15, twistLimitDegrees=8}
 {bone1Name=dyn_cape_cl_07, bone2Name=dyn_cape_cl_06, swing1LimitDegrees=10, swing2LimitDegrees=15, twistLimitDegrees=8}
 {bone1Name=dyn_cape_cr_MSTR, bone2Name=spine_05, swing1LimitDegrees=10, swing2LimitDegrees=15, twistLimitDegrees=8}
 {bone1Name=dyn_cape_cr_01, bone2Name=dyn_cape_cr_MSTR, swing1LimitDegrees=10, swing2LimitDegrees=15, twistLimitDegrees=8}
 {bone1Name=dyn_cape_cr_02, bone2Name=dyn_cape_cr_01, swing1LimitDegrees=10, swing2LimitDegrees=15, twistLimitDegrees=8}
 {bone1Name=dyn_cape_cr_03, bone2Name=dyn_cape_cr_02, swing1LimitDegrees=10, swing2LimitDegrees=15, twistLimitDegrees=8}
 {bone1Name=dyn_cape_cr_04, bone2Name=dyn_cape_cr_03, swing1LimitDegrees=10, swing2LimitDegrees=15, twistLimitDegrees=8}
 {bone1Name=dyn_cape_cr_05, bone2Name=dyn_cape_cr_04, swing1LimitDegrees=10, swing2LimitDegrees=15, twistLimitDegrees=8}
 {bone1Name=dyn_cape_cr_06, bone2Name=dyn_cape_cr_05, swing1LimitDegrees=10, swing2LimitDegrees=15, twistLimitDegrees=8}
 {bone1Name=dyn_cape_cr_07, bone2Name=dyn_cape_cr_06, swing1LimitDegrees=10, swing2LimitDegrees=15, twistLimitDegrees=8}
 {bone1Name=dyn_cape_l_MSTR, bone2Name=spine_05, swing1LimitDegrees=10, swing2LimitDegrees=15, twistLimitDegrees=8}
 {bone1Name=dyn_cape_l_01, bone2Name=dyn_cape_l_MSTR, swing1LimitDegrees=10, swing2LimitDegrees=15, twistLimitDegrees=8}
 {bone1Name=dyn_cape_l_02, bone2Name=dyn_cape_l_01, swing1LimitDegrees=10, swing2LimitDegrees=15, twistLimitDegrees=8}
 {bone1Name=dyn_cape_l_03, bone2Name=dyn_cape_l_02, swing1LimitDegrees=10, swing2LimitDegrees=15, twistLimitDegrees=8}
 {bone1Name=dyn_cape_l_04, bone2Name=dyn_cape_l_03, swing1LimitDegrees=10, swing2LimitDegrees=15, twistLimitDegrees=8}
 {bone1Name=dyn_cape_l_05, bone2Name=dyn_cape_l_04, swing1LimitDegrees=10, swing2LimitDegrees=15, twistLimitDegrees=8}
 {bone1Name=dyn_cape_l_06, bone2Name=dyn_cape_l_05, swing1LimitDegrees=10, swing2LimitDegrees=15, twistLimitDegrees=8}
 {bone1Name=dyn_cape_l_07, bone2Name=dyn_cape_l_06, swing1LimitDegrees=10, swing2LimitDegrees=15, twistLimitDegrees=8}
 {bone1Name=dyn_cape_r_MSTR, bone2Name=spine_05, swing1LimitDegrees=10, swing2LimitDegrees=15, twistLimitDegrees=8}
 {bone1Name=dyn_cape_r_01, bone2Name=dyn_cape_r_MSTR, swing1LimitDegrees=10, swing2LimitDegrees=15, twistLimitDegrees=8}
 {bone1Name=dyn_cape_r_02, bone2Name=dyn_cape_r_01, swing1LimitDegrees=10, swing2LimitDegrees=15, twistLimitDegrees=8}
 {bone1Name=dyn_cape_r_03, bone2Name=dyn_cape_r_02, swing1LimitDegrees=10, swing2LimitDegrees=15, twistLimitDegrees=8}
 {bone1Name=dyn_cape_r_04, bone2Name=dyn_cape_r_03, swing1LimitDegrees=10, swing2LimitDegrees=15, twistLimitDegrees=8}
 {bone1Name=dyn_cape_r_05, bone2Name=dyn_cape_r_04, swing1LimitDegrees=10, swing2LimitDegrees=15, twistLimitDegrees=8}
 {bone1Name=dyn_cape_r_06, bone2Name=dyn_cape_r_05, swing1LimitDegrees=10, swing2LimitDegrees=15, twistLimitDegrees=8}
 {
  bone1Name=pelvis
  bone2Name=root
  swing1LimitDegrees=45
  swing1Motion=Free
  swing2LimitDegrees=45
  swing2Motion=Free
  twistLimitDegrees=45
  twistMotion=Free
 }
 {
  bone1Name=thigh_l
  bone2Name=pelvis
  swing1LimitDegrees=45
  swing1Motion=Free
  swing2LimitDegrees=45
  swing2Motion=Free
  twistLimitDegrees=45
  twistMotion=Free
 }
 {
  bone1Name=calf_l
  bone2Name=thigh_l
  swing1LimitDegrees=45
  swing1Motion=Free
  swing2LimitDegrees=45
  swing2Motion=Free
  twistLimitDegrees=45
  twistMotion=Free
 }
 {
  bone1Name=thigh_r
  bone2Name=pelvis
  swing1LimitDegrees=45
  swing1Motion=Free
  swing2LimitDegrees=45
  swing2Motion=Free
  twistLimitDegrees=45
  twistMotion=Free
 }
 {
  bone1Name=calf_r
  bone2Name=thigh_r
  swing1LimitDegrees=45
  swing1Motion=Free
  swing2LimitDegrees=45
  swing2Motion=Free
  twistLimitDegrees=45
  twistMotion=Free
 }
 {
  bone1Name=upperarm_l
  bone2Name=spine_05
  swing1LimitDegrees=45
  swing1Motion=Free
  swing2LimitDegrees=45
  swing2Motion=Free
  twistLimitDegrees=45
  twistMotion=Free
 }
 {
  bone1Name=upperarm_r
  bone2Name=spine_05
  swing1LimitDegrees=45
  swing1Motion=Free
  swing2LimitDegrees=45
  swing2Motion=Free
  twistLimitDegrees=45
  twistMotion=Free
 }
]
