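# Animation Blueprint for the SheerWillUniversal cape accessory, adding secondary motion to four cape bone chains.
# The attached parent's pose gets world-space body bone overrides, then passes through a Control Rig and RigidBody simulation into a shared cache.
# AnimDynamics, trail and DBNO branches build on that cache, the latter two using rotation offsets and trail controllers weighted by BwdMovement.
# Booleans for rigid body, AnimDynamics and DBNO select the output.
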AnimBlueprint /BRCosmetics/Accessories/FORT_Capes/Cape_M_MED_SheerWillUniversal/Meshes/M_MED_SheerWillUniversal_Cape_AnimBP
target_skeleton=&/M_MED_SheerWillUniversal_Cape_Skeleton
graphs [
 AnimGraph {nodes [
   n1 = Blend Poses by bool(BlendPose_0=Use cached pose 'dbnoCache'(), BlendPose_1=Use cached pose 'trailCache'(), bActiveValue=Get bIsDBNO()) {blendType=Linear}
   n2 = Use cached pose 'AnimDyn Cache'()
   n3 = Get bAnimDynamicsEnabled() -> (bAnimDynamicsEnabled:bool)
   Output Pose(Result=Blend Poses by bool(
     BlendPose_0=Use cached pose 'RBANCache'()
     BlendPose_1=Blend Poses by bool(BlendPose_0=@n2, BlendPose_1=@n1, bActiveValue=@n3) {blendType=Linear}
     bActiveValue=Get bRigidBodyEnabled()
    ) {blendType=Linear}) {layerGroup=None}
   n4 = Get SkydivingAlpha() -> (SkydivingAlpha:Rotator)
   n5 = "Transform (Modify) Bone - Bone: dyn_cape_l_01"(ComponentPose=Local To Component(LocalPose=Use cached pose 'RBANCache'()), Alpha=1, Translation="0,0,0", Rotation=@n4, Scale="1,1,1") {rotationMode=BMM_Additive, rotationSpace=BCS_BoneSpace, alphaBoolBlend {blendOption=Linear}, boneToModify {boneName=dyn_cape_l_01}}
   n6 = "Transform (Modify) Bone - Bone: dyn_cape_cr_01"(ComponentPose=@n5, Alpha=1, Translation="0,0,0", Rotation=@n4, Scale="1,1,1") {rotationMode=BMM_Additive, rotationSpace=BCS_BoneSpace, alphaBoolBlend {blendOption=Linear}, boneToModify {boneName=dyn_cape_cr_01}}
   n7 = "Transform (Modify) Bone - Bone: dyn_cape_cl_01"(ComponentPose=@n6, Alpha=1, Translation="0,0,0", Rotation=@n4, Scale="1,1,1") {rotationMode=BMM_Additive, rotationSpace=BCS_BoneSpace, alphaBoolBlend {blendOption=Linear}, boneToModify {boneName=dyn_cape_cl_01}}
   n8 = "Transform (Modify) Bone - Bone: dyn_cape_r_01"(ComponentPose=@n7, Alpha=1, Translation="0,0,0", Rotation=@n4, Scale="1,1,1") {rotationMode=BMM_Additive, rotationSpace=BCS_BoneSpace, alphaBoolBlend {blendOption=Linear}, boneToModify {boneName=dyn_cape_r_01}}
   n9 = Get BwdMovement() -> (BwdMovement:double)
   n10 = Trail controller - Bone: dyn_cape_l_07(ComponentPose=@n8, Alpha=@n9) {
    bUsePlanarLimit=true
    chainLength=6
    debugLifeTime=0.1
    maxDeltaTime=0.0333
    alphaScaleBiasClamp {bInterpResult=true}
    baseJoint {boneName=backpack_root}
    planarLimits [{drivingBone {boneName=backpack_root}, planeTransform {location=<0 0 0>, rotation=<rot 0 0 -90>, scale=<1 1 1>}}]
    rotationLimits [
     {limitMax=<0 180 180>, limitMin=<0 -180 -180>}
     {limitMax=<0 180 180>, limitMin=<0 -180 -180>}
     {limitMax=<0 180 180>, limitMin=<0 -180 -180>}
     {limitMax=<180 180 180>}
     {limitMax=<180 180 180>}
     {limitMax=<180 180 180>}
    ]
    rotationOffsets [<0 0 0>, <0 0 0>, <0 0 0>, <0 0 0>, <0 0 0>, <0 0 0>]
    trailBone {boneName=dyn_cape_l_07}
    trailRelaxationSpeed {editorCurveData {keys [<at 0 23 RCIM_Linear>,<at 0.284 19.3799 RCIM_Linear>,<at 0.6554 8.5924 RCIM_Linear>,<at 1 5 RCIM_Linear>]}}
   }
   n11 = Trail controller - Bone: dyn_cape_cl_07(ComponentPose=@n10, Alpha=@n9) {
    bUsePlanarLimit=true
    chainLength=6
    debugLifeTime=0.1
    maxDeltaTime=0.0333
    alphaScaleBiasClamp {bInterpResult=true}
    baseJoint {boneName=backpack_root}
    planarLimits [{drivingBone {boneName=backpack_root}, planeTransform {location=<0 0 0>, rotation=<rot 0 0 -90>, scale=<1 1 1>}}]
    rotationLimits [
     {limitMax=<0 180 180>, limitMin=<0 -180 -180>}
     {limitMax=<0 180 180>, limitMin=<0 -180 -180>}
     {limitMax=<0 90 180>, limitMin=<0 -90 -180>}
     {limitMax=<180 180 180>}
     {limitMax=<180 180 180>}
     {limitMax=<180 180 180>}
    ]
    rotationOffsets [<0 0 0>, <0 0 0>, <0 0 0>, <0 0 0>, <0 0 0>, <0 0 0>]
    trailBone {boneName=dyn_cape_cl_07}
    trailRelaxationSpeed {editorCurveData {keys [<at 0 23 RCIM_Linear>,<at 0.284 19.3799 RCIM_Linear>,<at 0.6554 8.5924 RCIM_Linear>,<at 1 5 RCIM_Linear>]}}
   }
   n12 = Trail controller - Bone: dyn_cape_cr_07(ComponentPose=@n11, Alpha=@n9) {
    bUsePlanarLimit=true
    chainLength=6
    debugLifeTime=0.1
    maxDeltaTime=0.0333
    alphaScaleBiasClamp {bInterpResult=true}
    baseJoint {boneName=backpack_root}
    planarLimits [{drivingBone {boneName=backpack_root}, planeTransform {location=<0 0 0>, rotation=<rot 0 0 -90>, scale=<1 1 1>}}]
    rotationLimits [
     {limitMax=<0 180 180>, limitMin=<0 -180 -180>}
     {limitMax=<0 180 180>, limitMin=<0 -180 -180>}
     {limitMax=<0 90 180>, limitMin=<0 -90 -180>}
     {limitMax=<180 180 180>}
     {limitMax=<180 180 180>}
     {limitMax=<180 180 180>}
    ]
    rotationOffsets [<0 0 0>, <0 0 0>, <0 0 0>, <0 0 0>, <0 0 0>, <0 0 0>]
    trailBone {boneName=dyn_cape_cr_07}
    trailRelaxationSpeed {editorCurveData {keys [<at 0 23 RCIM_Linear>,<at 0.284 19.3799 RCIM_Linear>,<at 0.6554 8.5924 RCIM_Linear>,<at 1 5 RCIM_Linear>]}}
   }
   n13 = Trail controller - Bone: dyn_cape_r_07(ComponentPose=@n12, Alpha=@n9) {
    bUsePlanarLimit=true
    chainLength=6
    debugLifeTime=0.1
    maxDeltaTime=0.0333
    alphaScaleBiasClamp {bInterpResult=true}
    baseJoint {boneName=backpack_root}
    planarLimits [{drivingBone {boneName=backpack_root}, planeTransform {location=<0 0 0>, rotation=<rot 0 0 -90>, scale=<1 1 1>}}]
    rotationLimits [
     {limitMax=<0 180 180>, limitMin=<0 -180 -180>}
     {limitMax=<0 180 180>, limitMin=<0 -180 -180>}
     {limitMax=<0 90 180>, limitMin=<0 -90 -180>}
     {limitMax=<180 180 180>}
     {limitMax=<180 180 180>}
     {limitMax=<180 180 180>}
    ]
    rotationOffsets [<0 0 0>, <0 0 0>, <0 0 0>, <0 0 0>, <0 0 0>, <0 0 0>]
    trailBone {boneName=dyn_cape_r_07}
    trailRelaxationSpeed {editorCurveData {keys [<at 0 23 RCIM_Linear>,<at 0.284 19.3799 RCIM_Linear>,<at 0.6554 8.5924 RCIM_Linear>,<at 1 5 RCIM_Linear>]}}
   }
   Save cached pose 'trailCache'(Pose=Component To Local(ComponentPose=@n13)) {cacheName=trailCache}
   n14 = "Transform (Modify) Bone - Bone: dyn_cape_l_01"(ComponentPose=Local To Component(LocalPose=Use cached pose 'RBANCache'()), Alpha=1, Translation="0,0,0", Rotation="-10,0,0", Scale="1,1,1") {
    rotationMode=BMM_Additive
    rotationSpace=BCS_BoneSpace
    alphaBoolBlend {blendOption=Linear}
    boneToModify {boneName=dyn_cape_l_01}
    rotation {pitch=-10}
   }
   n15 = "Transform (Modify) Bone - Bone: dyn_cape_cl_01"(ComponentPose=@n14, Alpha=1, Translation="0,0,0", Rotation="-3.475,-0.010,-0.034", Scale="1,1,1") {
    rotationMode=BMM_Additive
    rotationSpace=BCS_BoneSpace
    alphaBoolBlend {blendOption=Linear}
    boneToModify {boneName=dyn_cape_cl_01}
    rotation=<rot -3.4751 -0.01 -0.034>
   }
   n16 = "Transform (Modify) Bone - Bone: dyn_cape_cr_01"(ComponentPose=@n15, Alpha=1, Translation="0,0,0", Rotation="-3.438,-0.006,0.052", Scale="1,1,1") {
    rotationMode=BMM_Additive
    rotationSpace=BCS_BoneSpace
    alphaBoolBlend {blendOption=Linear}
    boneToModify {boneName=dyn_cape_cr_01}
    rotation=<rot -3.438 -0.006 0.052>
   }
   n17 = "Transform (Modify) Bone - Bone: dyn_cape_r_01"(ComponentPose=@n16, Alpha=1, Translation="0,0,0", Rotation="-11.964,7.908,0.743", Scale="1,1,1") {
    rotationMode=BMM_Additive
    rotationSpace=BCS_BoneSpace
    alphaBoolBlend {blendOption=Linear}
    boneToModify {boneName=dyn_cape_r_01}
    rotation=<rot -11.964 7.908 0.743>
   }
   n18 = "Transform (Modify) Bone - Bone: dyn_cape_l_03"(ComponentPose=@n17, Alpha=1, Translation="0,0,0", Rotation="-10,0,0", Scale="1,1,1") {
    rotationMode=BMM_Additive
    rotationSpace=BCS_BoneSpace
    alphaBoolBlend {blendOption=Linear}
    boneToModify {boneName=dyn_cape_l_03}
    rotation {pitch=-10}
   }
   n19 = "Transform (Modify) Bone - Bone: dyn_cape_cl_03"(ComponentPose=@n18, Alpha=1, Translation="0,0,0", Rotation="-7,0,0", Scale="1,1,1") {
    rotationMode=BMM_Additive
    rotationSpace=BCS_BoneSpace
    alphaBoolBlend {blendOption=Linear}
    boneToModify {boneName=dyn_cape_cl_03}
    rotation {pitch=-7}
   }
   n20 = "Transform (Modify) Bone - Bone: dyn_cape_cr_03"(ComponentPose=@n19, Alpha=1, Translation="0,0,0", Rotation="-7,0,0", Scale="1,1,1") {
    rotationMode=BMM_Additive
    rotationSpace=BCS_BoneSpace
    alphaBoolBlend {blendOption=Linear}
    boneToModify {boneName=dyn_cape_cr_03}
    rotation {pitch=-7}
   }
   n21 = "Transform (Modify) Bone - Bone: dyn_cape_r_03"(ComponentPose=@n20, Alpha=1, Translation="0,0,0", Rotation="-3,0,0", Scale="1,1,1") {
    rotationMode=BMM_Additive
    rotationSpace=BCS_BoneSpace
    alphaBoolBlend {blendOption=Linear}
    boneToModify {boneName=dyn_cape_r_03}
    rotation {pitch=-3}
   }
   n22 = "Transform (Modify) Bone - Bone: dyn_cape_l_04"(ComponentPose=@n21, Alpha=1, Translation="0,0,0", Rotation="-10,0,0", Scale="1,1,1") {
    rotationMode=BMM_Additive
    rotationSpace=BCS_BoneSpace
    alphaBoolBlend {blendOption=Linear}
    boneToModify {boneName=dyn_cape_l_04}
    rotation {pitch=-10}
   }
   n23 = "Transform (Modify) Bone - Bone: dyn_cape_cl_04"(ComponentPose=@n22, Alpha=1, Translation="0,0,0", Rotation="-10,0,0", Scale="1,1,1") {
    rotationMode=BMM_Additive
    rotationSpace=BCS_BoneSpace
    alphaBoolBlend {blendOption=Linear}
    boneToModify {boneName=dyn_cape_cl_04}
    rotation {pitch=-10}
   }
   n24 = "Transform (Modify) Bone - Bone: dyn_cape_cr_04"(ComponentPose=@n23, Alpha=1, Translation="0,0,0", Rotation="-10,0,0", Scale="1,1,1") {
    rotationMode=BMM_Additive
    rotationSpace=BCS_BoneSpace
    alphaBoolBlend {blendOption=Linear}
    boneToModify {boneName=dyn_cape_cr_04}
    rotation {pitch=-10}
   }
   n25 = "Transform (Modify) Bone - Bone: dyn_cape_r_04"(ComponentPose=@n24, Alpha=1, Translation="0,0,0", Rotation="-20,0,0", Scale="1,1,1") {
    rotationMode=BMM_Additive
    rotationSpace=BCS_BoneSpace
    alphaBoolBlend {blendOption=Linear}
    boneToModify {boneName=dyn_cape_r_04}
    rotation {pitch=-20}
   }
   n26 = Get BwdMovement()
   n27 = Trail controller - Bone: dyn_cape_l_07(ComponentPose=@n25, Alpha=@n26) {
    bActorSpaceFakeVel=true
    bShowTrailLocation=true
    bUsePlanarLimit=true
    chainLength=8
    debugLifeTime=0.1
    maxDeltaTime=0.0333
    relaxationSpeedScale=2
    alphaScaleBiasClamp {bInterpResult=true, interpSpeedDecreasing=25, interpSpeedIncreasing=25}
    baseJoint {boneName=backpack_root}
    planarLimits [{drivingBone {boneName=None}, planeTransform {location=<0 0 10>, rotation=<rot 0 0 0>, scale=<1 1 1>}}]
    rotationLimits [
     {limitMax=<180 180 180>}
     {limitMax=<180 180 180>}
     {limitMax=<180 180 180>}
     {limitMax=<180 180 180>}
     {limitMax=<180 180 180>}
     {limitMax=<180 180 180>}
     {limitMax=<180 180 180>}
     {limitMax=<180 180 180>}
    ]
    rotationOffsets [<0 0 0>, <0 0 0>, <0 0 0>, <0 0 0>, <0 0 0>, <0 0 0>, <0 0 0>, <0 0 0>]
    trailBone {boneName=dyn_cape_l_07}
    trailRelaxationSpeed {editorCurveData {keys [<at 0 23 RCIM_Linear>,<at 0.284 19.3799 RCIM_Linear>,<at 0.6554 8.5924 RCIM_Linear>,<at 1 5 RCIM_Linear>]}}
   }
   n28 = Trail controller - Bone: dyn_cape_cl_07(ComponentPose=@n27, Alpha=@n26) {
    bActorSpaceFakeVel=true
    bShowTrailLocation=true
    bUsePlanarLimit=true
    chainLength=8
    debugLifeTime=0.1
    maxDeltaTime=0.0333
    relaxationSpeedScale=2
    alphaScaleBiasClamp {bInterpResult=true, interpSpeedDecreasing=25, interpSpeedIncreasing=25}
    baseJoint {boneName=backpack_root}
    planarLimits [{drivingBone {boneName=dyn_cape_cl_04}, planeTransform {location=<0 0 0>, rotation=<rot 0 0 0>, scale=<1 1 1>}}]
    rotationLimits [
     {limitMax=<180 180 180>}
     {limitMax=<180 180 180>}
     {limitMax=<180 180 180>}
     {limitMax=<180 180 180>}
     {limitMax=<180 180 180>}
     {limitMax=<180 180 180>}
     {limitMax=<180 180 180>}
     {limitMax=<180 180 180>}
    ]
    rotationOffsets [<0 0 0>, <0 0 0>, <0 0 0>, <0 0 0>, <0 0 0>, <0 0 0>, <0 0 0>, <0 0 0>]
    trailBone {boneName=dyn_cape_cl_07}
    trailRelaxationSpeed {editorCurveData {keys [<at 0 23 RCIM_Linear>,<at 0.284 19.3799 RCIM_Linear>,<at 0.6554 8.5924 RCIM_Linear>,<at 1 5 RCIM_Linear>]}}
   }
   n29 = Trail controller - Bone: dyn_cape_cr_07(ComponentPose=@n28, Alpha=@n26) {
    bActorSpaceFakeVel=true
    bShowTrailLocation=true
    bUsePlanarLimit=true
    chainLength=8
    debugLifeTime=0.1
    maxDeltaTime=0.0333
    relaxationSpeedScale=2
    alphaScaleBiasClamp {bInterpResult=true, interpSpeedDecreasing=25, interpSpeedIncreasing=25}
    baseJoint {boneName=backpack_root}
    planarLimits [{drivingBone {boneName=dyn_cape_cr_04}, planeTransform {location=<0 0 0>, rotation=<rot 0 0 0>, scale=<1 1 1>}}]
    rotationLimits [
     {limitMax=<180 180 180>}
     {limitMax=<180 180 180>}
     {limitMax=<180 180 180>}
     {limitMax=<180 180 180>}
     {limitMax=<180 180 180>}
     {limitMax=<180 180 180>}
     {limitMax=<180 180 180>}
     {limitMax=<180 180 180>}
    ]
    rotationOffsets [<0 0 0>, <0 0 0>, <0 0 0>, <0 0 0>, <0 0 0>, <0 0 0>, <0 0 0>, <0 0 0>]
    trailBone {boneName=dyn_cape_cr_07}
    trailRelaxationSpeed {editorCurveData {keys [<at 0 23 RCIM_Linear>,<at 0.284 19.3799 RCIM_Linear>,<at 0.6554 8.5924 RCIM_Linear>,<at 1 5 RCIM_Linear>]}}
   }
   Save cached pose 'dbnoCache'(Pose=Component To Local(ComponentPose=@n29)) {cacheName=dbnoCache}
   n30 = "Transform (Modify) Bone - Bone: root"(
    ComponentPose=Local To Component(LocalPose=Copy Pose From Mesh() {bUseAttachedParent=true})
    Translation=Get bodyRoot_tran() -> (bodyRoot_tran:Vector)
    Rotation=Get bodyRoot_rot() -> (bodyRoot_rot:Rotator)
   ) {
    rotationMode=BMM_Replace
    rotationSpace=BCS_WorldSpace
    scaleSpace=BCS_BoneSpace
    translationMode=BMM_Replace
    translationSpace=BCS_WorldSpace
    alphaBoolBlend {blendOption=Linear}
    alphaScaleBiasClamp {inRange {max=0, min=-40}, outRange {max=0, min=1}}
    boneToModify {boneName=root}
   }
   n31 = "Transform (Modify) Bone - Bone: pelvis"(
    ComponentPose=@n30
    Alpha=1
    Translation=Get pelvis_tran() -> (pelvis_tran:Vector)
    Rotation=Get pelvis_rot() -> (pelvis_rot:Rotator)
    Scale="1,1,1"
   ) {
    rotationMode=BMM_Replace
    rotationSpace=BCS_WorldSpace
    scaleSpace=BCS_BoneSpace
    translationMode=BMM_Replace
    translationSpace=BCS_WorldSpace
    alphaBoolBlend {blendOption=Linear}
    alphaScaleBiasClamp {inRange {max=0, min=-40}, outRange {max=0, min=1}}
    boneToModify {boneName=pelvis}
   }
   n32 = "Transform (Modify) Bone - Bone: cape_top"(
    ComponentPose=@n31
    Alpha=1
    Translation=Get root_trans() -> (root_trans:Vector)
    Rotation=Get root_rot() -> (root_rot:Rotator)
    Scale="1,1,1"
   ) {
    rotationMode=BMM_Replace
    rotationSpace=BCS_WorldSpace
    scaleSpace=BCS_BoneSpace
    translationMode=BMM_Replace
    translationSpace=BCS_WorldSpace
    alphaBoolBlend {blendOption=Linear}
    alphaScaleBiasClamp {inRange {max=0, min=-40}, outRange {max=0, min=1}}
    boneToModify {boneName=cape_top}
   }
   n33 = "Transform (Modify) Bone - Bone: thigh_l"(
    ComponentPose=@n32
    Alpha=1
    Translation=Get thigh_l_tran() -> (thigh_l_tran:Vector)
    Rotation=Get thigh_l_rot() -> (thigh_l_rot:Rotator)
    Scale="1,1,1"
   ) {
    rotationMode=BMM_Replace
    rotationSpace=BCS_WorldSpace
    scaleSpace=BCS_BoneSpace
    translationMode=BMM_Replace
    translationSpace=BCS_WorldSpace
    alphaBoolBlend {blendOption=Linear}
    alphaScaleBiasClamp {inRange {max=0, min=-40}, outRange {max=0, min=1}}
    boneToModify {boneName=thigh_l}
   }
   n34 = "Transform (Modify) Bone - Bone: thigh_r"(
    ComponentPose=@n33
    Alpha=1
    Translation=Get thigh_r_tran() -> (thigh_r_tran:Vector)
    Rotation=Get thigh_r_rot() -> (thigh_r_rot:Rotator)
    Scale="1,1,1"
   ) {
    rotationMode=BMM_Replace
    rotationSpace=BCS_WorldSpace
    scaleSpace=BCS_BoneSpace
    translationMode=BMM_Replace
    translationSpace=BCS_WorldSpace
    alphaBoolBlend {blendOption=Linear}
    alphaScaleBiasClamp {inRange {max=0, min=-40}, outRange {max=0, min=1}}
    boneToModify {boneName=thigh_r}
   }
   n35 = "Transform (Modify) Bone - Bone: calf_l"(
    ComponentPose=@n34
    Alpha=1
    Translation=Get calf_l_tran() -> (calf_l_tran:Vector)
    Rotation=Get calf_l_rot() -> (calf_l_rot:Rotator)
    Scale="1,1,1"
   ) {
    rotationMode=BMM_Replace
    rotationSpace=BCS_WorldSpace
    scaleSpace=BCS_BoneSpace
    translationMode=BMM_Replace
    translationSpace=BCS_WorldSpace
    alphaBoolBlend {blendOption=Linear}
    alphaScaleBiasClamp {inRange {max=0, min=-40}, outRange {max=0, min=1}}
    boneToModify {boneName=calf_l}
   }
   n36 = "Transform (Modify) Bone - Bone: calf_r"(
    ComponentPose=@n35
    Alpha=1
    Translation=Get calf_r_tran() -> (calf_r_tran:Vector)
    Rotation=Get calf_r_rot() -> (calf_r_rot:Rotator)
    Scale="1,1,1"
   ) {
    rotationMode=BMM_Replace
    rotationSpace=BCS_WorldSpace
    scaleSpace=BCS_BoneSpace
    translationMode=BMM_Replace
    translationSpace=BCS_WorldSpace
    alphaBoolBlend {blendOption=Linear}
    alphaScaleBiasClamp {inRange {max=0, min=-40}, outRange {max=0, min=1}}
    boneToModify {boneName=calf_r}
   }
   n37 = "Transform (Modify) Bone - Bone: spine_01"(
    ComponentPose=@n36
    Alpha=1
    Translation=Get spine_01_tran() -> (spine_01_tran:Vector)
    Rotation=Get spine_01_rot() -> (spine_01_rot:Rotator)
    Scale="1,1,1"
   ) {
    rotationMode=BMM_Replace
    rotationSpace=BCS_WorldSpace
    scaleSpace=BCS_BoneSpace
    translationMode=BMM_Replace
    translationSpace=BCS_WorldSpace
    alphaBoolBlend {blendOption=Linear}
    alphaScaleBiasClamp {inRange {max=0, min=-40}, outRange {max=0, min=1}}
    boneToModify {boneName=spine_01}
   }
   n38 = "Transform (Modify) Bone - Bone: spine_02"(
    ComponentPose=@n37
    Alpha=1
    Translation=Get spine_02_tran() -> (spine_02_tran:Vector)
    Rotation=Get spine_02_rot() -> (spine_02_rot:Rotator)
    Scale="1,1,1"
   ) {
    rotationMode=BMM_Replace
    rotationSpace=BCS_WorldSpace
    scaleSpace=BCS_BoneSpace
    translationMode=BMM_Replace
    translationSpace=BCS_WorldSpace
    alphaBoolBlend {blendOption=Linear}
    alphaScaleBiasClamp {inRange {max=0, min=-40}, outRange {max=0, min=1}}
    boneToModify {boneName=spine_02}
   }
   n39 = "Transform (Modify) Bone - Bone: spine_03"(
    ComponentPose=@n38
    Alpha=1
    Translation=Get spine_03_tran() -> (spine_03_tran:Vector)
    Rotation=Get spine_03_rot() -> (spine_03_rot:Rotator)
    Scale="1,1,1"
   ) {
    rotationMode=BMM_Replace
    rotationSpace=BCS_WorldSpace
    scaleSpace=BCS_BoneSpace
    translationMode=BMM_Replace
    translationSpace=BCS_WorldSpace
    alphaBoolBlend {blendOption=Linear}
    alphaScaleBiasClamp {inRange {max=0, min=-40}, outRange {max=0, min=1}}
    boneToModify {boneName=spine_03}
   }
   n40 = "Transform (Modify) Bone - Bone: spine_04"(
    ComponentPose=@n39
    Alpha=1
    Translation=Get spine_04_tran() -> (spine_04_tran:Vector)
    Rotation=Get spine_04_rot() -> (spine_04_rot:Rotator)
    Scale="1,1,1"
   ) {
    rotationMode=BMM_Replace
    rotationSpace=BCS_WorldSpace
    scaleSpace=BCS_BoneSpace
    translationMode=BMM_Replace
    translationSpace=BCS_WorldSpace
    alphaBoolBlend {blendOption=Linear}
    alphaScaleBiasClamp {inRange {max=0, min=-40}, outRange {max=0, min=1}}
    boneToModify {boneName=spine_04}
   }
   n41 = "Transform (Modify) Bone - Bone: spine_05"(
    ComponentPose=@n40
    Alpha=1
    Translation=Get spine_05_tran() -> (spine_05_tran:Vector)
    Rotation=Get spine_05_rot() -> (spine_05_rot:Rotator)
    Scale="1,1,1"
   ) {
    rotationMode=BMM_Replace
    rotationSpace=BCS_WorldSpace
    scaleSpace=BCS_BoneSpace
    translationMode=BMM_Replace
    translationSpace=BCS_WorldSpace
    alphaBoolBlend {blendOption=Linear}
    alphaScaleBiasClamp {inRange {max=0, min=-40}, outRange {max=0, min=1}}
    boneToModify {boneName=spine_05}
   }
   n42 = "Transform (Modify) Bone - Bone: clavicle_r"(
    ComponentPose=@n41
    Alpha=1
    Translation=Get clavicle_r_tran() -> (clavicle_r_tran:Vector)
    Rotation=Get clavicel_r_rot() -> (clavicel_r_rot:Rotator)
    Scale="1,1,1"
   ) {
    rotationMode=BMM_Replace
    rotationSpace=BCS_WorldSpace
    scaleSpace=BCS_BoneSpace
    translationMode=BMM_Replace
    translationSpace=BCS_WorldSpace
    alphaBoolBlend {blendOption=Linear}
    alphaScaleBiasClamp {inRange {max=0, min=-40}, outRange {max=0, min=1}}
    boneToModify {boneName=clavicle_r}
   }
   n43 = "Transform (Modify) Bone - Bone: clavicle_l"(
    ComponentPose=@n42
    Alpha=1
    Translation=Get clavicle_l_tran() -> (clavicle_l_tran:Vector)
    Rotation=Get clavicle_l_rot() -> (clavicle_l_rot:Rotator)
    Scale="1,1,1"
   ) {
    rotationMode=BMM_Replace
    rotationSpace=BCS_WorldSpace
    scaleSpace=BCS_BoneSpace
    translationMode=BMM_Replace
    translationSpace=BCS_WorldSpace
    alphaBoolBlend {blendOption=Linear}
    alphaScaleBiasClamp {inRange {max=0, min=-40}, outRange {max=0, min=1}}
    boneToModify {boneName=clavicle_l}
   }
   n44 = "Transform (Modify) Bone - Bone: upperarm_r"(
    ComponentPose=@n43
    Alpha=1
    Translation=Get upperarm_r_tran() -> (upperarm_r_tran:Vector)
    Rotation=Get upperarm_r_rot() -> (upperarm_r_rot:Rotator)
    Scale="1,1,1"
   ) {
    rotationMode=BMM_Replace
    rotationSpace=BCS_WorldSpace
    scaleSpace=BCS_BoneSpace
    translationMode=BMM_Replace
    translationSpace=BCS_WorldSpace
    alphaBoolBlend {blendOption=Linear}
    alphaScaleBiasClamp {inRange {max=0, min=-40}, outRange {max=0, min=1}}
    boneToModify {boneName=upperarm_r}
   }
   n45 = "Transform (Modify) Bone - Bone: upperarm_l"(
    ComponentPose=@n44
    Alpha=1
    Translation=Get upperarm_l_tran() -> (upperarm_l_tran:Vector)
    Rotation=Get upperarm_l_rot() -> (upperarm_l_rot:Rotator)
    Scale="1,1,1"
   ) {
    rotationMode=BMM_Replace
    rotationSpace=BCS_WorldSpace
    scaleSpace=BCS_BoneSpace
    translationMode=BMM_Replace
    translationSpace=BCS_WorldSpace
    alphaBoolBlend {blendOption=Linear}
    alphaScaleBiasClamp {inRange {max=0, min=-40}, outRange {max=0, min=1}}
    boneToModify {boneName=upperarm_l}
   }
   n46 = "Transform (Modify) Bone - Bone: lowerarm_l"(
    ComponentPose=@n45
    Alpha=1
    Translation=Get lowerarm_l_tran() -> (lowerarm_l_tran:Vector)
    Rotation=Get lowerarm_l_rot() -> (lowerarm_l_rot:Rotator)
    Scale="1,1,1"
   ) {
    rotationMode=BMM_Replace
    rotationSpace=BCS_WorldSpace
    scaleSpace=BCS_BoneSpace
    translationMode=BMM_Replace
    translationSpace=BCS_WorldSpace
    alphaBoolBlend {blendOption=Linear}
    alphaScaleBiasClamp {inRange {max=0, min=-40}, outRange {max=0, min=1}}
    boneToModify {boneName=lowerarm_l}
   }
   n47 = "Transform (Modify) Bone - Bone: lowerarm_r"(
    ComponentPose=@n46
    Alpha=1
    Translation=Get lowerarm_r_tran() -> (lowerarm_r_tran:Vector)
    Rotation=Get lowerarm_r_rot() -> (lowerarm_r_rot:Rotator)
    Scale="1,1,1"
   ) {
    rotationMode=BMM_Replace
    rotationSpace=BCS_WorldSpace
    scaleSpace=BCS_BoneSpace
    translationMode=BMM_Replace
    translationSpace=BCS_WorldSpace
    alphaBoolBlend {blendOption=Linear}
    alphaScaleBiasClamp {inRange {max=0, min=-40}, outRange {max=0, min=1}}
    boneToModify {boneName=lowerarm_r}
   }
   n48 = Control Rig(Source=Component To Local(ComponentPose=@n47)) {controlRigAssetReference {controlRigAsset=&/M_MED_SheerWillUniversal_Cape_CtrlRig}}
   n49 = "Transform (Modify) Bone - Bone: dyn_cape_l_MSTR"!(ComponentPose=Local To Component(LocalPose=@n48), bAlphaBoolEnabled:bool=True, Translation="0,0,0", Rotation="35,0,0", Scale="1,1,1") -> (Pose) {
    alphaInputType=Bool
    rotationMode=BMM_Additive
    rotationSpace=BCS_BoneSpace
    alphaBoolBlend {blendOption=Linear}
    boneToModify {boneName=dyn_cape_l_MSTR}
    rotation {pitch=35}
   }
   n50 = "Transform (Modify) Bone - Bone: dyn_cape_cl_MSTR"!(ComponentPose=@n49, bAlphaBoolEnabled:bool=True, Translation="0,0,0", Rotation="35,0,0", Scale="1,1,1") -> (Pose) {
    alphaInputType=Bool
    rotationMode=BMM_Additive
    rotationSpace=BCS_BoneSpace
    alphaBoolBlend {blendOption=Linear}
    boneToModify {boneName=dyn_cape_cl_MSTR}
    rotation {pitch=35}
   }
   n51 = "Transform (Modify) Bone - Bone: dyn_cape_cr_MSTR"!(ComponentPose=@n50, bAlphaBoolEnabled:bool=True, Translation="0,0,0", Rotation="35,0,0", Scale="1,1,1") -> (Pose) {
    alphaInputType=Bool
    rotationMode=BMM_Additive
    rotationSpace=BCS_BoneSpace
    alphaBoolBlend {blendOption=Linear}
    boneToModify {boneName=dyn_cape_cr_MSTR}
    rotation {pitch=35}
   }
   n52 = "Transform (Modify) Bone - Bone: dyn_cape_r_MSTR"!(ComponentPose=@n51, bAlphaBoolEnabled:bool=True, Translation="0,0,0", Rotation="35,0,0", Scale="1,1,1") -> (Pose) {
    alphaInputType=Bool
    rotationMode=BMM_Additive
    rotationSpace=BCS_BoneSpace
    alphaBoolBlend {blendOption=Linear}
    boneToModify {boneName=dyn_cape_r_MSTR}
    rotation {pitch=35}
   }
   n53 = RigidBody!(
    ComponentPose=@n52
    Alpha
    OverrideWorldGravity:Vector=Get GravityOverride()
    SimSpaceSettings:Sim Space Settings Structure=Get Sim Space Settings() -> (Sim Space Settings:Sim Space Settings Structure)
   ) -> (Pose) {
    bClampLinearTranslationLimitToRefPose=true
    bOverrideWorldGravity=true
    bUseExternalClothCollision=true
    overridePhysicsAsset=&/M_MED_SheerWillUniversal_Cape_Physics
    simulationSpace=BaseBoneSpace
    baseBoneRef {boneName=backpack_root}
    componentAppliedLinearAccClamp=<6000 6000 6000>
   }
   Save cached pose 'RBANCache'(Pose=Component To Local(ComponentPose=@n53)) {cacheName=RBANCache}
   n54 = Anim Dynamics - Bone: dyn_cape_cl_01(ComponentPose=Local To Component(LocalPose=Use cached pose 'RBANCache'()), Alpha=1, ExternalForce="0,0,0") {
    angularDampingOverride=1
    angularSpringConstant=5000
    bAngularSpring=true
    bOverrideAngularDamping=true
    bUseGravityOverride=true
    gravityScale=0.35
    alphaBoolBlend {blendOption=Linear}
    boundBone {boneName=dyn_cape_cl_01}
    componentAppliedLinearAccClamp=<3000 3000 3000>
    componentLinearAccScale=<2 2 2>
    componentLinearVelScale=<2 2 2>
    gravityOverride {z=-1500}
    physicsBodyDefinitions [{
      boundBone {boneName=dyn_cape_cl_01}
      boxExtents=<10 25 25>
      constraintSetup {angularLimitsMax=<0 10 15>, angularLimitsMin=<0 -40 -15>, angularTarget=<1 0 0>}
      localJointOffset=<1.5 0 0>
     }]
    relativeSpaceBone {boneName=root}
   }
   n55 = Anim Dynamics - Bone: dyn_cape_cl_02!(ComponentPose=@n54, Alpha=1, GravityOverride:Vector="0,0,-980", ExternalForce="0,0,0") -> (Pose) {
    angularDampingOverride=1
    angularSpringConstant=5000
    bAngularSpring=true
    bOverrideAngularDamping=true
    bUseGravityOverride=true
    gravityScale=0.35
    alphaBoolBlend {blendOption=Linear}
    boundBone {boneName=dyn_cape_cl_02}
    componentAppliedLinearAccClamp=<3000 3000 3000>
    componentLinearAccScale=<2 2 2>
    componentLinearVelScale=<2 2 2>
    gravityOverride {z=-1500}
    physicsBodyDefinitions [{
      boundBone {boneName=dyn_cape_cl_02}
      boxExtents=<10 25 25>
      constraintSetup {angularLimitsMax=<0 10 15>, angularLimitsMin=<0 -40 -15>, angularTarget=<1 0 0>}
      localJointOffset=<1.5 0 0>
     }]
    relativeSpaceBone {boneName=root}
   }
   n56 = Anim Dynamics - Bone: dyn_cape_cl_03(ComponentPose=@n55, Alpha=1, ExternalForce="0,0,0") {
    angularDampingOverride=1
    angularSpringConstant=5000
    bAngularSpring=true
    bOverrideAngularDamping=true
    bUseGravityOverride=true
    gravityScale=0.35
    alphaBoolBlend {blendOption=Linear}
    boundBone {boneName=dyn_cape_cl_03}
    componentAppliedLinearAccClamp=<3000 3000 3000>
    componentLinearAccScale=<2 2 2>
    componentLinearVelScale=<2 2 2>
    gravityOverride {z=-1500}
    physicsBodyDefinitions [{
      boundBone {boneName=dyn_cape_cl_03}
      boxExtents=<10 25 25>
      constraintSetup {angularLimitsMax=<0 10 15>, angularLimitsMin=<0 -40 -15>, angularTarget=<1 0 0>}
      localJointOffset=<1.5 0 0>
     }]
    relativeSpaceBone {boneName=root}
   }
   n57 = Anim Dynamics - Bone: dyn_cape_cl_04(ComponentPose=@n56, Alpha=1, ExternalForce="0,0,0") {
    angularDampingOverride=1
    angularSpringConstant=5000
    bAngularSpring=true
    bOverrideAngularDamping=true
    bUseGravityOverride=true
    gravityScale=0.35
    alphaBoolBlend {blendOption=Linear}
    boundBone {boneName=dyn_cape_cl_04}
    componentAppliedLinearAccClamp=<3000 3000 3000>
    componentLinearAccScale=<2 2 2>
    componentLinearVelScale=<2 2 2>
    gravityOverride {z=-1500}
    physicsBodyDefinitions [{
      boundBone {boneName=dyn_cape_cl_04}
      boxExtents=<10 25 25>
      constraintSetup {angularLimitsMax=<0 10 15>, angularLimitsMin=<0 -40 -15>, angularTarget=<1 0 0>}
      localJointOffset=<1.5 0 0>
     }]
    relativeSpaceBone {boneName=root}
   }
   n58 = Anim Dynamics - Bone: dyn_cape_cl_05(ComponentPose=@n57, Alpha=1, ExternalForce="0,0,0") {
    angularDampingOverride=1
    angularSpringConstant=5000
    bAngularSpring=true
    bOverrideAngularDamping=true
    bUseGravityOverride=true
    gravityScale=0.35
    alphaBoolBlend {blendOption=Linear}
    boundBone {boneName=dyn_cape_cl_05}
    componentAppliedLinearAccClamp=<3000 3000 3000>
    componentLinearAccScale=<2 2 2>
    componentLinearVelScale=<2 2 2>
    gravityOverride {z=-1500}
    physicsBodyDefinitions [{
      boundBone {boneName=dyn_cape_cl_05}
      boxExtents=<10 25 25>
      constraintSetup {angularLimitsMax=<0 10 15>, angularLimitsMin=<0 -40 -15>, angularTarget=<1 0 0>}
      localJointOffset=<1.5 0 0>
     }]
    relativeSpaceBone {boneName=root}
   }
   n59 = Anim Dynamics - Bone: dyn_cape_cl_06(ComponentPose=@n58, Alpha=1, ExternalForce="0,0,0") {
    angularDampingOverride=1
    angularSpringConstant=5000
    bAngularSpring=true
    bOverrideAngularDamping=true
    bUseGravityOverride=true
    gravityScale=0.35
    alphaBoolBlend {blendOption=Linear}
    boundBone {boneName=dyn_cape_cl_06}
    componentAppliedLinearAccClamp=<3000 3000 3000>
    componentLinearAccScale=<2 2 2>
    componentLinearVelScale=<2 2 2>
    gravityOverride {z=-1500}
    physicsBodyDefinitions [{
      boundBone {boneName=dyn_cape_cl_06}
      boxExtents=<10 25 25>
      constraintSetup {angularLimitsMax=<0 10 15>, angularLimitsMin=<0 -40 -15>, angularTarget=<1 0 0>}
      localJointOffset=<1.5 0 0>
     }]
    relativeSpaceBone {boneName=root}
   }
   n60 = Anim Dynamics - Bone: dyn_cape_cl_07(ComponentPose=@n59, Alpha=1, ExternalForce="0,0,0") {
    angularDampingOverride=1
    angularSpringConstant=5000
    bAngularSpring=true
    bOverrideAngularDamping=true
    bUseGravityOverride=true
    gravityScale=0.35
    alphaBoolBlend {blendOption=Linear}
    boundBone {boneName=dyn_cape_cl_07}
    componentAppliedLinearAccClamp=<3000 3000 3000>
    componentLinearAccScale=<2 2 2>
    componentLinearVelScale=<2 2 2>
    gravityOverride {z=-1500}
    physicsBodyDefinitions [{
      boundBone {boneName=dyn_cape_cl_07}
      boxExtents=<10 25 25>
      constraintSetup {angularLimitsMax=<0 10 15>, angularLimitsMin=<0 -40 -15>, angularTarget=<1 0 0>}
      localJointOffset=<1.5 0 0>
     }]
    relativeSpaceBone {boneName=root}
   }
   n61 = Anim Dynamics - Bone: dyn_cape_cr_01(ComponentPose=@n60, Alpha=1, ExternalForce="0,0,0") {
    angularDampingOverride=1
    angularSpringConstant=5000
    bAngularSpring=true
    bOverrideAngularDamping=true
    bUseGravityOverride=true
    gravityScale=0.35
    alphaBoolBlend {blendOption=Linear}
    boundBone {boneName=dyn_cape_cr_01}
    componentAppliedLinearAccClamp=<3000 3000 3000>
    componentLinearAccScale=<2 2 2>
    componentLinearVelScale=<2 2 2>
    gravityOverride {z=-1500}
    physicsBodyDefinitions [{
      boundBone {boneName=dyn_cape_cr_01}
      boxExtents=<10 23 23>
      constraintSetup {angularLimitsMax=<0 5 15>, angularLimitsMin=<0 -40 -15>, angularTarget=<1 0 0>}
      localJointOffset=<1 0 0>
     }]
    relativeSpaceBone {boneName=root}
   }
   n62 = Anim Dynamics - Bone: dyn_cape_cr_02(ComponentPose=@n61, Alpha=1, ExternalForce="0,0,0") {
    angularDampingOverride=1
    angularSpringConstant=5000
    bAngularSpring=true
    bOverrideAngularDamping=true
    bUseGravityOverride=true
    gravityScale=0.35
    alphaBoolBlend {blendOption=Linear}
    boundBone {boneName=dyn_cape_cr_02}
    componentAppliedLinearAccClamp=<3000 3000 3000>
    componentLinearAccScale=<2 2 2>
    componentLinearVelScale=<2 2 2>
    gravityOverride {z=-1500}
    physicsBodyDefinitions [{
      boundBone {boneName=dyn_cape_cr_02}
      boxExtents=<10 23 23>
      constraintSetup {angularLimitsMax=<0 5 15>, angularLimitsMin=<0 -40 -15>, angularTarget=<1 0 0>}
      localJointOffset=<1 0 0>
     }]
    relativeSpaceBone {boneName=root}
   }
   n63 = Anim Dynamics - Bone: dyn_cape_cr_03(ComponentPose=@n62, Alpha=1, ExternalForce="0,0,0") {
    angularDampingOverride=1
    angularSpringConstant=5000
    bAngularSpring=true
    bOverrideAngularDamping=true
    bUseGravityOverride=true
    gravityScale=0.35
    alphaBoolBlend {blendOption=Linear}
    boundBone {boneName=dyn_cape_cr_03}
    componentAppliedLinearAccClamp=<3000 3000 3000>
    componentLinearAccScale=<2 2 2>
    componentLinearVelScale=<2 2 2>
    gravityOverride {z=-1500}
    physicsBodyDefinitions [{
      boundBone {boneName=dyn_cape_cr_03}
      boxExtents=<10 23 23>
      constraintSetup {angularLimitsMax=<0 5 15>, angularLimitsMin=<0 -40 -15>, angularTarget=<1 0 0>}
      localJointOffset=<1 0 0>
     }]
    relativeSpaceBone {boneName=root}
   }
   n64 = Anim Dynamics - Bone: dyn_cape_cr_04(ComponentPose=@n63, Alpha=1, ExternalForce="0,0,0") {
    angularDampingOverride=1
    angularSpringConstant=5000
    bAngularSpring=true
    bOverrideAngularDamping=true
    bUseGravityOverride=true
    gravityScale=0.35
    alphaBoolBlend {blendOption=Linear}
    boundBone {boneName=dyn_cape_cr_04}
    componentAppliedLinearAccClamp=<3000 3000 3000>
    componentLinearAccScale=<2 2 2>
    componentLinearVelScale=<2 2 2>
    gravityOverride {z=-1500}
    physicsBodyDefinitions [{
      boundBone {boneName=dyn_cape_cr_04}
      boxExtents=<10 23 23>
      constraintSetup {angularLimitsMax=<0 5 15>, angularLimitsMin=<0 -40 -15>, angularTarget=<1 0 0>}
      localJointOffset=<1 0 0>
     }]
    relativeSpaceBone {boneName=root}
   }
   n65 = Anim Dynamics - Bone: dyn_cape_cr_05(ComponentPose=@n64, Alpha=1, ExternalForce="0,0,0") {
    angularDampingOverride=1
    angularSpringConstant=5000
    bAngularSpring=true
    bOverrideAngularDamping=true
    bUseGravityOverride=true
    gravityScale=0.35
    alphaBoolBlend {blendOption=Linear}
    boundBone {boneName=dyn_cape_cr_05}
    componentAppliedLinearAccClamp=<3000 3000 3000>
    componentLinearAccScale=<2 2 2>
    componentLinearVelScale=<2 2 2>
    gravityOverride {z=-1500}
    physicsBodyDefinitions [{
      boundBone {boneName=dyn_cape_cr_05}
      boxExtents=<10 23 23>
      constraintSetup {angularLimitsMax=<0 5 15>, angularLimitsMin=<0 -40 -15>, angularTarget=<1 0 0>}
      localJointOffset=<1 0 0>
     }]
    relativeSpaceBone {boneName=root}
   }
   n66 = Anim Dynamics - Bone: dyn_cape_cr_06(ComponentPose=@n65, Alpha=1, ExternalForce="0,0,0") {
    angularDampingOverride=1
    angularSpringConstant=5000
    bAngularSpring=true
    bOverrideAngularDamping=true
    bUseGravityOverride=true
    gravityScale=0.35
    alphaBoolBlend {blendOption=Linear}
    boundBone {boneName=dyn_cape_cr_06}
    componentAppliedLinearAccClamp=<3000 3000 3000>
    componentLinearAccScale=<2 2 2>
    componentLinearVelScale=<2 2 2>
    gravityOverride {z=-1500}
    physicsBodyDefinitions [{
      boundBone {boneName=dyn_cape_cr_06}
      boxExtents=<10 23 23>
      constraintSetup {angularLimitsMax=<0 5 15>, angularLimitsMin=<0 -40 -15>, angularTarget=<1 0 0>}
      localJointOffset=<1 0 0>
     }]
    relativeSpaceBone {boneName=root}
   }
   n67 = Anim Dynamics - Bone: dyn_cape_l_07(ComponentPose=@n66, Alpha=1, ExternalForce="0,0,0") {
    angularDampingOverride=1
    angularSpringConstant=5000
    bAngularSpring=true
    bOverrideAngularDamping=true
    bUseGravityOverride=true
    gravityScale=0.35
    alphaBoolBlend {blendOption=Linear}
    boundBone {boneName=dyn_cape_l_07}
    componentAppliedLinearAccClamp=<3000 3000 3000>
    componentLinearAccScale=<2 2 2>
    componentLinearVelScale=<2 2 2>
    gravityOverride {z=-1500}
    physicsBodyDefinitions [{
      boundBone {boneName=dyn_cape_l_07}
      boxExtents=<10 23 23>
      constraintSetup {angularLimitsMax=<0 5 15>, angularLimitsMin=<0 -40 -15>, angularTarget=<1 0 0>}
      localJointOffset=<1 0 0>
     }]
    relativeSpaceBone {boneName=root}
   }
   n68 = Anim Dynamics - Bone: dyn_cape_r_01(ComponentPose=@n67, Alpha=1, ExternalForce="0,0,0") {
    angularDampingOverride=1
    angularSpringConstant=5000
    bAngularSpring=true
    bOverrideAngularDamping=true
    bUseGravityOverride=true
    gravityScale=0.35
    alphaBoolBlend {blendOption=Linear}
    boundBone {boneName=dyn_cape_r_01}
    componentAppliedLinearAccClamp=<3000 3000 3000>
    componentLinearAccScale=<2 2 2>
    componentLinearVelScale=<2 2 2>
    gravityOverride {z=-1500}
    physicsBodyDefinitions [{
      boundBone {boneName=dyn_cape_r_01}
      boxExtents=<10 22 22>
      constraintSetup {angularLimitsMax=<0 5 15>, angularLimitsMin=<0 -40 -15>, angularTarget=<1 0 0>}
      localJointOffset=<1.2 0 0>
     }]
    relativeSpaceBone {boneName=root}
   }
   n69 = Anim Dynamics - Bone: dyn_cape_r_02(ComponentPose=@n68, Alpha=1, ExternalForce="0,0,0") {
    angularDampingOverride=1
    angularSpringConstant=5000
    bAngularSpring=true
    bOverrideAngularDamping=true
    bUseGravityOverride=true
    gravityScale=0.35
    alphaBoolBlend {blendOption=Linear}
    boundBone {boneName=dyn_cape_r_02}
    componentAppliedLinearAccClamp=<3000 3000 3000>
    componentLinearAccScale=<2 2 2>
    componentLinearVelScale=<2 2 2>
    gravityOverride {z=-1500}
    physicsBodyDefinitions [{
      boundBone {boneName=dyn_cape_r_02}
      boxExtents=<10 22 22>
      constraintSetup {angularLimitsMax=<0 5 15>, angularLimitsMin=<0 -40 -15>, angularTarget=<1 0 0>}
      localJointOffset=<1.2 0 0>
     }]
    relativeSpaceBone {boneName=root}
   }
   n70 = Anim Dynamics - Bone: dyn_cape_r_03(ComponentPose=@n69, Alpha=1, ExternalForce="0,0,0") {
    angularDampingOverride=1
    angularSpringConstant=5000
    bAngularSpring=true
    bOverrideAngularDamping=true
    bUseGravityOverride=true
    gravityScale=0.35
    alphaBoolBlend {blendOption=Linear}
    boundBone {boneName=dyn_cape_r_03}
    componentAppliedLinearAccClamp=<3000 3000 3000>
    componentLinearAccScale=<2 2 2>
    componentLinearVelScale=<2 2 2>
    gravityOverride {z=-1500}
    physicsBodyDefinitions [{
      boundBone {boneName=dyn_cape_r_03}
      boxExtents=<10 22 22>
      constraintSetup {angularLimitsMax=<0 5 15>, angularLimitsMin=<0 -40 -15>, angularTarget=<1 0 0>}
      localJointOffset=<1.2 0 0>
     }]
    relativeSpaceBone {boneName=root}
   }
   n71 = Anim Dynamics - Bone: dyn_cape_r_04(ComponentPose=@n70, Alpha=1, ExternalForce="0,0,0") {
    angularDampingOverride=1
    angularSpringConstant=5000
    bAngularSpring=true
    bOverrideAngularDamping=true
    bUseGravityOverride=true
    gravityScale=0.35
    alphaBoolBlend {blendOption=Linear}
    boundBone {boneName=dyn_cape_r_04}
    componentAppliedLinearAccClamp=<3000 3000 3000>
    componentLinearAccScale=<2 2 2>
    componentLinearVelScale=<2 2 2>
    gravityOverride {z=-1500}
    physicsBodyDefinitions [{
      boundBone {boneName=dyn_cape_r_04}
      boxExtents=<10 22 22>
      constraintSetup {angularLimitsMax=<0 5 15>, angularLimitsMin=<0 -40 -15>, angularTarget=<1 0 0>}
      localJointOffset=<1.2 0 0>
     }]
    relativeSpaceBone {boneName=root}
   }
   n72 = Anim Dynamics - Bone: dyn_cape_r_05(ComponentPose=@n71, Alpha=1, ExternalForce="0,0,0") {
    angularDampingOverride=1
    angularSpringConstant=5000
    bAngularSpring=true
    bOverrideAngularDamping=true
    bUseGravityOverride=true
    gravityScale=0.35
    alphaBoolBlend {blendOption=Linear}
    boundBone {boneName=dyn_cape_r_05}
    componentAppliedLinearAccClamp=<3000 3000 3000>
    componentLinearAccScale=<2 2 2>
    componentLinearVelScale=<2 2 2>
    gravityOverride {z=-1500}
    physicsBodyDefinitions [{
      boundBone {boneName=dyn_cape_r_05}
      boxExtents=<10 22 22>
      constraintSetup {angularLimitsMax=<0 5 15>, angularLimitsMin=<0 -40 -15>, angularTarget=<1 0 0>}
      localJointOffset=<1.2 0 0>
     }]
    relativeSpaceBone {boneName=root}
   }
   n73 = Anim Dynamics - Bone: dyn_cape_r_06(ComponentPose=@n72, Alpha=1, ExternalForce="0,0,0") {
    angularDampingOverride=1
    angularSpringConstant=5000
    bAngularSpring=true
    bOverrideAngularDamping=true
    bUseGravityOverride=true
    gravityScale=0.35
    alphaBoolBlend {blendOption=Linear}
    boundBone {boneName=dyn_cape_r_06}
    componentAppliedLinearAccClamp=<3000 3000 3000>
    componentLinearAccScale=<2 2 2>
    componentLinearVelScale=<2 2 2>
    gravityOverride {z=-1500}
    physicsBodyDefinitions [{
      boundBone {boneName=dyn_cape_r_06}
      boxExtents=<10 22 22>
      constraintSetup {angularLimitsMax=<0 5 15>, angularLimitsMin=<0 -40 -15>, angularTarget=<1 0 0>}
      localJointOffset=<1.2 0 0>
     }]
    relativeSpaceBone {boneName=root}
   }
   n74 = Anim Dynamics - Bone: dyn_cape_r_07(ComponentPose=@n73, Alpha=1, ExternalForce="0,0,0") {
    angularDampingOverride=1
    angularSpringConstant=5000
    bAngularSpring=true
    bOverrideAngularDamping=true
    bUseGravityOverride=true
    gravityScale=0.35
    alphaBoolBlend {blendOption=Linear}
    boundBone {boneName=dyn_cape_r_07}
    componentAppliedLinearAccClamp=<3000 3000 3000>
    componentLinearAccScale=<2 2 2>
    componentLinearVelScale=<2 2 2>
    gravityOverride {z=-1500}
    physicsBodyDefinitions [{
      boundBone {boneName=dyn_cape_r_07}
      boxExtents=<10 22 22>
      constraintSetup {angularLimitsMax=<0 5 15>, angularLimitsMin=<0 -40 -15>, angularTarget=<1 0 0>}
      localJointOffset=<1.2 0 0>
     }]
    relativeSpaceBone {boneName=root}
   }
   n75 = Anim Dynamics - Bone: dyn_cape_l_01(ComponentPose=@n74, Alpha=1, ExternalForce="0,0,0") {
    angularDampingOverride=1
    angularSpringConstant=5000
    bAngularSpring=true
    bOverrideAngularDamping=true
    bUseGravityOverride=true
    gravityScale=0.35
    alphaBoolBlend {blendOption=Linear}
    boundBone {boneName=dyn_cape_l_01}
    componentAppliedLinearAccClamp=<3000 3000 3000>
    componentLinearAccScale=<2 2 2>
    componentLinearVelScale=<2 2 2>
    gravityOverride {z=-1500}
    physicsBodyDefinitions [{
      boundBone {boneName=dyn_cape_l_01}
      boxExtents=<10 25 25>
      constraintSetup {angularLimitsMax=<0 5 15>, angularLimitsMin=<0 -40 -15>, angularTarget=<1 0 0>}
      localJointOffset=<1 0 0>
     }]
    relativeSpaceBone {boneName=root}
   }
   n76 = Anim Dynamics - Bone: dyn_cape_l_02(ComponentPose=@n75, Alpha=1, ExternalForce="0,0,0") {
    angularDampingOverride=1
    angularSpringConstant=5000
    bAngularSpring=true
    bOverrideAngularDamping=true
    bUseGravityOverride=true
    gravityScale=0.35
    alphaBoolBlend {blendOption=Linear}
    boundBone {boneName=dyn_cape_l_02}
    componentAppliedLinearAccClamp=<3000 3000 3000>
    componentLinearAccScale=<2 2 2>
    componentLinearVelScale=<2 2 2>
    gravityOverride {z=-1500}
    physicsBodyDefinitions [{
      boundBone {boneName=dyn_cape_l_02}
      boxExtents=<10 25 25>
      constraintSetup {angularLimitsMax=<0 5 15>, angularLimitsMin=<0 -40 -15>, angularTarget=<1 0 0>}
      localJointOffset=<1 0 0>
     }]
    relativeSpaceBone {boneName=root}
   }
   n77 = Anim Dynamics - Bone: dyn_cape_l_03(ComponentPose=@n76, Alpha=1, ExternalForce="0,0,0") {
    angularDampingOverride=1
    angularSpringConstant=5000
    bAngularSpring=true
    bOverrideAngularDamping=true
    bUseGravityOverride=true
    gravityScale=0.35
    alphaBoolBlend {blendOption=Linear}
    boundBone {boneName=dyn_cape_l_03}
    componentAppliedLinearAccClamp=<3000 3000 3000>
    componentLinearAccScale=<2 2 2>
    componentLinearVelScale=<2 2 2>
    gravityOverride {z=-1500}
    physicsBodyDefinitions [{
      boundBone {boneName=dyn_cape_l_03}
      boxExtents=<10 25 25>
      constraintSetup {angularLimitsMax=<0 5 15>, angularLimitsMin=<0 -40 -15>, angularTarget=<1 0 0>}
      localJointOffset=<1 0 0>
     }]
    relativeSpaceBone {boneName=root}
   }
   n78 = Anim Dynamics - Bone: dyn_cape_l_04(ComponentPose=@n77, Alpha=1, ExternalForce="0,0,0") {
    angularDampingOverride=1
    angularSpringConstant=5000
    bAngularSpring=true
    bOverrideAngularDamping=true
    bUseGravityOverride=true
    gravityScale=0.35
    alphaBoolBlend {blendOption=Linear}
    boundBone {boneName=dyn_cape_l_04}
    componentAppliedLinearAccClamp=<3000 3000 3000>
    componentLinearAccScale=<2 2 2>
    componentLinearVelScale=<2 2 2>
    gravityOverride {z=-1500}
    physicsBodyDefinitions [{
      boundBone {boneName=dyn_cape_l_04}
      boxExtents=<10 25 25>
      constraintSetup {angularLimitsMax=<0 5 15>, angularLimitsMin=<0 -40 -15>, angularTarget=<1 0 0>}
      localJointOffset=<1 0 0>
     }]
    relativeSpaceBone {boneName=root}
   }
   n79 = Anim Dynamics - Bone: dyn_cape_l_05(ComponentPose=@n78, Alpha=1, ExternalForce="0,0,0") {
    angularDampingOverride=1
    angularSpringConstant=5000
    bAngularSpring=true
    bOverrideAngularDamping=true
    bUseGravityOverride=true
    gravityScale=0.35
    alphaBoolBlend {blendOption=Linear}
    boundBone {boneName=dyn_cape_l_05}
    componentAppliedLinearAccClamp=<3000 3000 3000>
    componentLinearAccScale=<2 2 2>
    componentLinearVelScale=<2 2 2>
    gravityOverride {z=-1500}
    physicsBodyDefinitions [{
      boundBone {boneName=dyn_cape_l_05}
      boxExtents=<10 25 25>
      constraintSetup {angularLimitsMax=<0 5 15>, angularLimitsMin=<0 -40 -15>, angularTarget=<1 0 0>}
      localJointOffset=<1 0 0>
     }]
    relativeSpaceBone {boneName=root}
   }
   n80 = Anim Dynamics - Bone: dyn_cape_l_06(ComponentPose=@n79, Alpha=1, ExternalForce="0,0,0") {
    angularDampingOverride=1
    angularSpringConstant=5000
    bAngularSpring=true
    bOverrideAngularDamping=true
    bUseGravityOverride=true
    gravityScale=0.35
    alphaBoolBlend {blendOption=Linear}
    boundBone {boneName=dyn_cape_l_06}
    componentAppliedLinearAccClamp=<3000 3000 3000>
    componentLinearAccScale=<2 2 2>
    componentLinearVelScale=<2 2 2>
    gravityOverride {z=-1500}
    physicsBodyDefinitions [{
      boundBone {boneName=dyn_cape_l_06}
      boxExtents=<10 25 25>
      constraintSetup {angularLimitsMax=<0 5 15>, angularLimitsMin=<0 -40 -15>, angularTarget=<1 0 0>}
      localJointOffset=<1 0 0>
     }]
    relativeSpaceBone {boneName=root}
   }
   n81 = Anim Dynamics - Bone: dyn_cape_l_07(ComponentPose=@n80, Alpha=1, ExternalForce="0,0,0") {
    angularDampingOverride=1
    angularSpringConstant=5000
    bAngularSpring=true
    bOverrideAngularDamping=true
    bUseGravityOverride=true
    gravityScale=0.35
    alphaBoolBlend {blendOption=Linear}
    boundBone {boneName=dyn_cape_l_07}
    componentAppliedLinearAccClamp=<3000 3000 3000>
    componentLinearAccScale=<2 2 2>
    componentLinearVelScale=<2 2 2>
    gravityOverride {z=-1500}
    physicsBodyDefinitions [{
      boundBone {boneName=dyn_cape_l_07}
      boxExtents=<10 25 25>
      constraintSetup {angularLimitsMax=<0 5 15>, angularLimitsMin=<0 -40 -15>, angularTarget=<1 0 0>}
      localJointOffset=<1 0 0>
     }]
    relativeSpaceBone {boneName=root}
   }
   Save cached pose 'AnimDyn Cache'(Pose=Component To Local(ComponentPose=@n81)) {cacheName=AnimDyn Cache}
   n82 = Get BwdMovement()
   n83 = Trail controller - Bone: dyn_cape_l_07(Alpha=@n82) {
    bUsePlanarLimit=true
    chainLength=7
    debugLifeTime=0.1
    maxDeltaTime=0.0333
    alphaScaleBiasClamp {bInterpResult=true}
    baseJoint {boneName=backpack_root}
    planarLimits [{drivingBone {boneName=backpack_root}, planeTransform {location=<0 0 0>, rotation=<rot 0 0 -90>, scale=<1 1 1>}}]
    rotationLimits [
     {limitMax=<0 180 180>, limitMin=<0 -180 -180>}
     {limitMax=<0 180 180>, limitMin=<0 -180 -180>}
     {limitMax=<0 180 180>, limitMin=<0 -180 -180>}
     {limitMax=<0 180 180>, limitMin=<0 -180 -180>}
     {limitMax=<0 180 180>, limitMin=<0 -180 -180>}
     {limitMax=<0 180 180>, limitMin=<0 -180 -180>}
     {limitMax=<0 180 180>, limitMin=<0 -180 -180>}
    ]
    rotationOffsets [<0 0 0>, <0 0 0>, <0 0 0>, <0 0 0>, <0 0 0>, <0 0 0>, <0 0 0>]
    trailBone {boneName=dyn_cape_l_07}
    trailRelaxationSpeed {editorCurveData {keys [<at 0 23 RCIM_Linear>,<at 0.284 19.3799 RCIM_Linear>,<at 0.6554 8.5924 RCIM_Linear>,<at 1 5 RCIM_Linear>]}}
   }
   n84 = Trail controller - Bone: dyn_cape_cl_07(ComponentPose=@n83, Alpha=@n82) {
    bUsePlanarLimit=true
    chainLength=7
    debugLifeTime=0.1
    maxDeltaTime=0.0333
    alphaScaleBiasClamp {bInterpResult=true}
    baseJoint {boneName=backpack_root}
    planarLimits [{drivingBone {boneName=backpack_root}, planeTransform {location=<0 0 0>, rotation=<rot 0 0 -90>, scale=<1 1 1>}}]
    rotationLimits [
     {limitMax=<0 180 180>, limitMin=<0 -180 -180>}
     {limitMax=<0 180 180>, limitMin=<0 -180 -180>}
     {limitMax=<0 90 180>, limitMin=<0 -90 -180>}
     {limitMax=<0 180 180>, limitMin=<0 -180 -180>}
     {limitMax=<0 180 180>, limitMin=<0 -180 -180>}
     {limitMax=<0 180 180>, limitMin=<0 -180 -180>}
     {limitMax=<0 180 180>, limitMin=<0 -180 -180>}
    ]
    rotationOffsets [<0 0 0>, <0 0 0>, <0 0 0>, <0 0 0>, <0 0 0>, <0 0 0>, <0 0 0>]
    trailBone {boneName=dyn_cape_cl_07}
    trailRelaxationSpeed {editorCurveData {keys [<at 0 23 RCIM_Linear>,<at 0.284 19.3799 RCIM_Linear>,<at 0.6554 8.5924 RCIM_Linear>,<at 1 5 RCIM_Linear>]}}
   }
   n85 = Trail controller - Bone: dyn_cape_cr_07(ComponentPose=@n84, Alpha=@n82) {
    bUsePlanarLimit=true
    chainLength=7
    debugLifeTime=0.1
    maxDeltaTime=0.0333
    alphaScaleBiasClamp {bInterpResult=true}
    baseJoint {boneName=backpack_root}
    planarLimits [{drivingBone {boneName=backpack_root}, planeTransform {location=<0 0 0>, rotation=<rot 0 0 -90>, scale=<1 1 1>}}]
    rotationLimits [
     {limitMax=<0 180 180>, limitMin=<0 -180 -180>}
     {limitMax=<0 180 180>, limitMin=<0 -180 -180>}
     {limitMax=<0 90 180>, limitMin=<0 -90 -180>}
     {limitMax=<0 180 180>, limitMin=<0 -180 -180>}
     {limitMax=<0 180 180>, limitMin=<0 -180 -180>}
     {limitMax=<0 180 180>, limitMin=<0 -180 -180>}
     {limitMax=<0 180 180>, limitMin=<0 -180 -180>}
    ]
    rotationOffsets [<0 0 0>, <0 0 0>, <0 0 0>, <0 0 0>, <0 0 0>, <0 0 0>, <0 0 0>]
    trailBone {boneName=dyn_cape_cr_07}
    trailRelaxationSpeed {editorCurveData {keys [<at 0 23 RCIM_Linear>,<at 0.284 19.3799 RCIM_Linear>,<at 0.6554 8.5924 RCIM_Linear>,<at 1 5 RCIM_Linear>]}}
   }
   Trail controller - Bone: dyn_cape_r_07(ComponentPose=@n85, Alpha=@n82) {
    bUsePlanarLimit=true
    chainLength=7
    debugLifeTime=0.1
    maxDeltaTime=0.0333
    alphaScaleBiasClamp {bInterpResult=true}
    baseJoint {boneName=backpack_root}
    planarLimits [{drivingBone {boneName=backpack_root}, planeTransform {location=<0 0 0>, rotation=<rot 0 0 -90>, scale=<1 1 1>}}]
    rotationLimits [
     {limitMax=<0 180 180>, limitMin=<0 -180 -180>}
     {limitMax=<0 180 180>, limitMin=<0 -180 -180>}
     {limitMax=<0 90 180>, limitMin=<0 -90 -180>}
     {limitMax=<0 180 180>, limitMin=<0 -180 -180>}
     {limitMax=<0 180 180>, limitMin=<0 -180 -180>}
     {limitMax=<0 180 180>, limitMin=<0 -180 -180>}
     {limitMax=<0 180 180>, limitMin=<0 -180 -180>}
    ]
    rotationOffsets [<0 0 0>, <0 0 0>, <0 0 0>, <0 0 0>, <0 0 0>, <0 0 0>, <0 0 0>]
    trailBone {boneName=dyn_cape_r_07}
    trailRelaxationSpeed {editorCurveData {keys [<at 0 23 RCIM_Linear>,<at 0.284 19.3799 RCIM_Linear>,<at 0.6554 8.5924 RCIM_Linear>,<at 1 5 RCIM_Linear>]}}
   }
   Trail controller - Bone: dyn_cape_r_07(Alpha=@n26) {
    bActorSpaceFakeVel=true
    bShowTrailLocation=true
    bUsePlanarLimit=true
    chainLength=8
    debugLifeTime=0.1
    maxDeltaTime=0.0333
    relaxationSpeedScale=2
    alphaScaleBiasClamp {bInterpResult=true, interpSpeedDecreasing=25, interpSpeedIncreasing=25}
    baseJoint {boneName=backpack_root}
    planarLimits [{drivingBone {boneName=pelvis}, planeTransform {location=<0 0 16>, rotation=<rot 0 0 -40>, scale=<1 1 1>}}]
    rotationLimits [
     {limitMax=<180 180 180>}
     {limitMax=<180 180 180>}
     {limitMax=<180 180 180>}
     {limitMax=<180 180 180>}
     {limitMax=<180 180 180>}
     {limitMax=<180 180 180>}
     {limitMax=<180 180 180>}
     {limitMax=<180 180 180>}
    ]
    rotationOffsets [<0 0 0>, <0 0 0>, <0 0 0>, <0 0 0>, <0 0 0>, <0 0 0>, <0 0 0>, <0 0 0>]
    trailBone {boneName=dyn_cape_r_07}
    trailRelaxationSpeed {editorCurveData {keys [<at 0 23 RCIM_Linear>,<at 0.284 19.3799 RCIM_Linear>,<at 0.6554 8.5924 RCIM_Linear>,<at 1 5 RCIM_Linear>]}}
   }
   Get isEmotingAlpha() -> (isEmotingAlpha:double)
  ]}
 EventGraph {nodes [
   Event BlueprintUpdateAnimation()
   > n1 = Sequence!(execute) -> (then_0, then_1, then_2:Exec, then_3:Exec, then_4:Exec, then_5:Exec)
   > n2 = "Gravity Override State Function (Parameter Asset) (Nativized)"(
    CharacterPartAnimInstance=Self-Reference()
    GameTimeInSeconds=GetGameTimeInSeconds()
    DebugJointName=spine_05
    Parameters=&/M_MED_SheerWillUniversal_Cape_Parameters
   )
   > Set GravityOverride(GravityOverride=@n2)
   n3 = Break Fort Rigid Body Anim Node Parameters!(FortRigidBodyAnimNodeParameters:"Fort Rigid Body Anim Node Parameters Structure (by ref)"=@n2.OutRigidBodyAnimNodeParameters) -> (ComponentLinearAccScale:Vector=~, ComponentLinearVelScale:Vector=~, SimSpaceSettings:Sim Space Settings Structure) {is=title K2Node_BreakStruct}
   > Set Sim Space Settings(Sim Space Settings:Sim Space Settings Structure=@n3.SimSpaceSettings) -> (Output_Get:Sim Space Settings Structure)
   > n4 = PlayerMovement(AnimInstance=Self-Reference(), XDotProductCoefficient=1, YDotProductCoefficient=1, ZDotProductCoefficient=1)
   n5 = Break Sim Space Settings!(SimSpaceSettings:"Sim Space Settings Structure (by ref)"=@n3.SimSpaceSettings) -> (
    WorldAlpha:float=~
    VelocityScaleZ:float=~
    DampingAlpha:float=~
    MaxLinearVelocity:float=~
    MaxAngularVelocity:float=~
    MaxLinearAcceleration:float=~
    MaxAngularAcceleration:float=~
    ExternalLinearDragV:Vector=~
    ExternalLinearVelocity:Vector=~
    ExternalAngularVelocity:Vector=~
   ) {is=title K2Node_BreakStruct}
   n6 = MapRangeClamped(Value=@n4.YMovement, InRangeA=0, InRangeB=-1, OutRangeA=3, OutRangeB=0.002)
   > Set Sim Space Settings(Sim Space Settings=Make Sim Space Settings!(
     WorldAlpha:float=@n5
     VelocityScaleZ:float=@n5.VelocityScaleZ
     DampingAlpha:float=@n5.DampingAlpha
     MaxLinearVelocity:float=@n5.MaxLinearVelocity
     MaxAngularVelocity:float=@n5.MaxAngularVelocity
     MaxLinearAcceleration:float=@n5.MaxLinearAcceleration
     MaxAngularAcceleration:float=@n5.MaxAngularAcceleration
     ExternalLinearDragV:Vector=MakeVector(Y=@n6)
     ExternalLinearVelocity:Vector=@n5.ExternalLinearVelocity
     ExternalAngularVelocity:Vector=@n5.ExternalAngularVelocity
    ) -> (SimSpaceSettings:Sim Space Settings Structure) {is=title K2Node_MakeStruct})
   n7 = Branch(execute=@n1.then_1, Condition=Get bIsInFrontEnd())
   > Set BwdMovement(BwdMovement:double=0.5) -> (Output_Get:double)
   n8 = PlayerMovement(execute=@n7.else, AnimInstance=Self-Reference(), XDotProductCoefficient=1, YDotProductCoefficient=1, ZDotProductCoefficient=1)
   > n9 = Branch(Condition=Get bIsDBNOCarried())
   > Set BwdMovement(BwdMovement=0.5)
   n10 = Branch(execute=@n9.else, Condition=AND Boolean(A=Get bIsTargeting(), B=Get bIsCrouching()))
   n11 = MapRangeClamped(
    Value=float + float(A="Absolute (Float)"(A=@n8), B=MapRangeClamped(Value=@n8.YMovement, InRangeA=-1, InRangeB=1, OutRangeA=0.3)) {is=title K2Node_CommutativeAssociativeBinaryOperator}
    InRangeA=0
    InRangeB=1.65
   )
   > Set BwdMovement(BwdMovement=float * float(A=@n11, B=0.25) {is=title K2Node_PromotableOperator})
   Set BwdMovement(execute=@n10.else, BwdMovement=@n11)
   n12 = Branch(execute=@n1.then_2, Condition=Get bIsSkyDiving())
   > n13 = Branch(Condition="float > float"(A=Get PawnVelocityZ(), B=50) {is=title K2Node_CallFunction})
   > Set BwdMovement(BwdMovement=float * float(A=Get BwdMovement(), B=0) {is=title K2Node_CommutativeAssociativeBinaryOperator})
   Set BwdMovement(execute=[@n12.else, @n13.else], BwdMovement=float * float(A=Get BwdMovement(), B=1))
   n14 = Cast To FortPlayerPawn(execute=@n1.then_3, Object=TryGetPawnOwner())
   > n15 = Branch(Condition=Get bIsOnGrindRail())
   > n16 = Branch(Condition=NOT Boolean(A=IsWeaponHolstered!(self:Fort Pawn Object Reference=@n14) -> (ReturnValue:bool)))
   > n17 = Branch(Condition=Get bIsTargeting())
   > Set Sim Space Settings(Sim Space Settings=Get Sim Space Settings_GrindTarget() -> (Sim Space Settings_GrindTarget:Sim Space Settings Structure))
   Set Sim Space Settings(
    execute=@n17.else
    Sim Space Settings=Get Sim Space Settings_GrindingWeaponActive() -> (Sim Space Settings_GrindingWeaponActive:Sim Space Settings Structure)
   )
   Set Sim Space Settings(execute=[@n15.else, @n16.else], Sim Space Settings=Get Sim Space Settings())
   n18 = GetPartSkeletalMeshComponent()
   Is Valid(exec=@n1.then_4, InputObject=@n18)
   n19 = GetAttachParent!(self:Scene Component Object Reference=@n18) -> (ReturnValue:Scene Component Object Reference)
   > Is Valid(InputObject=@n19)
   > n20 = Sequence!(execute) -> (
    then_0
    then_1
    then_2
    then_3
    then_4
    then_5
    then_6:Exec
    then_7:Exec
    then_8:Exec
    then_9:Exec
    then_10:Exec
    then_11:Exec
    then_12:Exec
    then_13:Exec
    then_14:Exec
    then_15:Exec
    then_16:Exec
    then_17:Exec
    then_18:Exec
    then_19:Exec
   )
   n21 = BreakTransform(InTransform=GetSocketTransform(self=@n19, InSocketName=spine_05, TransformSpace=RTS_World))
   Set spine_05_tran(execute=@n20.then_10, spine_05_tran:Vector=@n21) -> (Output_Get:Vector)
   > Set spine_05_rot(spine_05_rot:Rotator=@n21.Rotation) -> (Output_Get:Rotator)
   n22 = BreakTransform(InTransform=GetSocketTransform(self=@n19, InSocketName=clavicle_r, TransformSpace=RTS_World))
   Set clavicle_r_tran(execute=@n20.then_11, clavicle_r_tran:Vector=@n22) -> (Output_Get:Vector)
   > Set clavicel_r_rot(clavicel_r_rot:Rotator=@n22.Rotation) -> (Output_Get:Rotator)
   n23 = BreakTransform(InTransform=GetSocketTransform(self=@n19, InSocketName=clavicle_l, TransformSpace=RTS_World))
   Set clavicle_l_tran(execute=@n20.then_12, clavicle_l_tran:Vector=@n23) -> (Output_Get:Vector)
   > Set clavicle_l_rot(clavicle_l_rot:Rotator=@n23.Rotation) -> (Output_Get:Rotator)
   n24 = BreakTransform(InTransform=GetSocketTransform(self=@n19, InSocketName=upperarm_r, TransformSpace=RTS_World))
   Set upperarm_r_tran(execute=@n20.then_13, upperarm_r_tran:Vector=@n24) -> (Output_Get:Vector)
   > Set upperarm_r_rot(upperarm_r_rot:Rotator=@n24.Rotation) -> (Output_Get:Rotator)
   n25 = BreakTransform(InTransform=GetSocketTransform(self=@n19, InSocketName=upperarm_l, TransformSpace=RTS_World))
   Set upperarm_l_tran(execute=@n20.then_14, upperarm_l_tran:Vector=@n25) -> (Output_Get:Vector)
   > Set upperarm_l_rot(upperarm_l_rot:Rotator=@n25.Rotation) -> (Output_Get:Rotator)
   n26 = BreakTransform(InTransform=GetSocketTransform(self=@n19, InSocketName=lowerarm_l, TransformSpace=RTS_World))
   Set lowerarm_l_tran(execute=@n20.then_15, lowerarm_l_tran:Vector=@n26) -> (Output_Get:Vector)
   > Set lowerarm_l_rot(lowerarm_l_rot:Rotator=@n26.Rotation) -> (Output_Get:Rotator)
   n27 = BreakTransform(InTransform=GetSocketTransform(self=@n19, InSocketName=lowerarm_r, TransformSpace=RTS_World))
   Set lowerarm_r_tran(execute=@n20.then_16, lowerarm_r_tran:Vector=@n27) -> (Output_Get:Vector)
   > Set lowerarm_r_rot(lowerarm_r_rot:Rotator=@n27.Rotation) -> (Output_Get:Rotator)
   n28 = BreakTransform(InTransform=GetSocketTransform(self=@n19, InSocketName=attach_cape, TransformSpace=RTS_World))
   Set root_trans(execute=@n20.then_17, root_trans:Vector=@n28) -> (Output_Get:Vector)
   > Set root_rot(root_rot:Rotator=@n28.Rotation) -> (Output_Get:Rotator)
   n29 = BreakTransform(InTransform=GetSocketTransform(self=@n19, InSocketName=spine_01, TransformSpace=RTS_World))
   Set spine_01_tran(execute=@n20.then_6, spine_01_tran:Vector=@n29) -> (Output_Get:Vector)
   > Set spine_01_rot(spine_01_rot:Rotator=@n29.Rotation) -> (Output_Get:Rotator)
   n30 = BreakTransform(InTransform=GetSocketTransform(self=@n19, InSocketName=spine_02, TransformSpace=RTS_World))
   Set spine_02_tran(execute=@n20.then_7, spine_02_tran:Vector=@n30) -> (Output_Get:Vector)
   > Set spine_02_rot(spine_02_rot:Rotator=@n30.Rotation) -> (Output_Get:Rotator)
   n31 = BreakTransform(InTransform=GetSocketTransform(self=@n19, InSocketName=spine_03, TransformSpace=RTS_World))
   Set spine_03_tran(execute=@n20.then_8, spine_03_tran:Vector=@n31) -> (Output_Get:Vector)
   > Set spine_03_rot(spine_03_rot:Rotator=@n31.Rotation) -> (Output_Get:Rotator)
   n32 = BreakTransform(InTransform=GetSocketTransform(self=@n19, InSocketName=spine_04, TransformSpace=RTS_World))
   Set spine_04_tran(execute=@n20.then_9, spine_04_tran:Vector=@n32) -> (Output_Get:Vector)
   > Set spine_04_rot(spine_04_rot:Rotator=@n32.Rotation) -> (Output_Get:Rotator)
   n33 = BreakTransform(InTransform=GetSocketTransform(self=@n19, InSocketName=pelvis, TransformSpace=RTS_World))
   Set pelvis_tran(execute=@n20.then_0, pelvis_tran:Vector=@n33) -> (Output_Get:Vector)
   > Set pelvis_rot(pelvis_rot:Rotator=@n33.Rotation) -> (Output_Get:Rotator)
   n34 = BreakTransform(InTransform=GetSocketTransform(self=@n19, InSocketName=thigh_l, TransformSpace=RTS_World))
   Set thigh_l_tran(execute=@n20.then_1, thigh_l_tran:Vector=@n34) -> (Output_Get:Vector)
   > Set thigh_l_rot(thigh_l_rot:Rotator=@n34.Rotation) -> (Output_Get:Rotator)
   n35 = BreakTransform(InTransform=GetSocketTransform(self=@n19, InSocketName=thigh_r, TransformSpace=RTS_World))
   Set thigh_r_tran(execute=@n20.then_2, thigh_r_tran:Vector=@n35) -> (Output_Get:Vector)
   > Set thigh_r_rot(thigh_r_rot:Rotator=@n35.Rotation) -> (Output_Get:Rotator)
   n36 = BreakTransform(InTransform=GetSocketTransform(self=@n19, InSocketName=calf_r, TransformSpace=RTS_World))
   Set calf_r_tran(execute=@n20.then_4, calf_r_tran:Vector=@n36) -> (Output_Get:Vector)
   > Set calf_r_rot(calf_r_rot:Rotator=@n36.Rotation) -> (Output_Get:Rotator)
   n37 = BreakTransform(InTransform=GetSocketTransform(self=@n19, InSocketName=calf_l, TransformSpace=RTS_World))
   Set calf_l_tran(execute=@n20.then_3, calf_l_tran:Vector=@n37) -> (Output_Get:Vector)
   > Set calf_l_rot(calf_l_rot:Rotator=@n37.Rotation) -> (Output_Get:Rotator)
   n38 = BreakTransform(InTransform=GetSocketTransform(self=@n19, InSocketName=root, TransformSpace=RTS_World))
   Set bodyRoot_tran(execute=@n20.then_5, bodyRoot_tran:Vector=@n38) -> (Output_Get:Vector)
   > Set bodyRoot_rot(bodyRoot_rot:Rotator=@n38.Rotation) -> (Output_Get:Rotator)
   Event OnBeginSkydiving()
   > Set SkydivingAlpha(SkydivingAlpha:Rotator="65,0,0") -> (Output_Get:Rotator)
   Event OnEndSkydiving()
   > Set SkydivingAlpha(SkydivingAlpha="10,-10,0")
   Set Is LOD2-3(Is LOD2-3:bool="integer >= integer"(A=Get CurrentLOD(), B=2)) -> (Output_Get:bool)
  ]}
]
variables [
 bodyRoot_rot {type=Rotator}
 bodyRoot_tran {type=Vector}
 BwdMovement {type=number}
 calf_l_rot {type=Rotator}
 calf_l_tran {type=Vector}
 calf_r_rot {type=Rotator}
 calf_r_tran {type=Vector}
 clavicel_r_rot {type=Rotator}
 clavicle_l_rot {type=Rotator}
 clavicle_l_tran {type=Vector}
 clavicle_r_tran {type=Vector}
 GravityOverride {type=Vector}
 Is LOD2-3 {type=boolean}
 lowerarm_l_rot {type=Rotator}
 lowerarm_l_tran {type=Vector}
 lowerarm_r_rot {type=Rotator}
 lowerarm_r_tran {type=Vector}
 pelvis_rot {type=Rotator}
 pelvis_tran {type=Vector}
 RBAN Settings Out {type=FortRigidBodyAnimNodeParameters}
 RBANAcceleration {type=Vector}
 RBANVelocity {type=Vector}
 root_rot {type=Rotator}
 root_trans {type=Vector}
 Sim Space Settings {
  type=SimSpaceSettings
  default {
   dampingAlpha=1
   maxAngularAcceleration=10000
   maxAngularVelocity=10000
   maxLinearAcceleration=10000
   maxLinearVelocity=10000
   velocityScaleZ=1
   worldAlpha=0
   externalLinearDragV=<0 0 0>
  }
 }
 Sim Space Settings_GrindingWeaponActive {
  type=SimSpaceSettings
  default {
   dampingAlpha=1
   maxAngularAcceleration=10000
   maxAngularVelocity=10000
   maxLinearAcceleration=10000
   maxLinearVelocity=10000
   velocityScaleZ=1
   worldAlpha=0
   externalLinearDragV=<0 0 0>
  }
 }
 Sim Space Settings_GrindTarget {
  type=SimSpaceSettings
  default {
   dampingAlpha=1
   maxAngularAcceleration=10000
   maxAngularVelocity=10000
   maxLinearAcceleration=10000
   maxLinearVelocity=10000
   velocityScaleZ=1
   worldAlpha=0
   externalLinearDragV=<0 0 0>
  }
 }
 SkydivingAlpha {type=Rotator}
 spine_01_rot {type=Rotator}
 spine_01_tran {type=Vector}
 spine_02_rot {type=Rotator}
 spine_02_tran {type=Vector}
 spine_03_rot {type=Rotator}
 spine_03_tran {type=Vector}
 spine_04_rot {type=Rotator}
 spine_04_tran {type=Vector}
 spine_05_rot {type=Rotator}
 spine_05_tran {type=Vector}
 thigh_l_rot {type=Rotator}
 thigh_l_tran {type=Vector}
 thigh_r_rot {type=Rotator}
 thigh_r_tran {type=Vector}
 upperarm_l_rot {type=Rotator}
 upperarm_l_tran {type=Vector}
 upperarm_r_rot {type=Rotator}
 upperarm_r_tran {type=Vector}
]
